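; Spawner Blueprint for Reality Saplings that overrides SpawnActor to create a sapling carrying saved state.
; If the class to spawn casts to the sapling base, it spawns the hardcoded base sapling class, passing in growth, watering, weeding, fruit, persistence and planter data; otherwise it defers to the parent.
; The sapling spawn then dispatches ServerActorSpawned.
; Point, radial and any-damage events only print debug strings, with any-damage also mislabeled as radial.
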(event EventBeginPlay
  (|CallParentFunction) ; This node is disabled and will not be called.\nDrag off pins to build functionality.
)

(event Collision|EventActorBeginOverlap
  (params (OtherActor :type Actor%20Object%20Reference)))

(event EventTick
  (params (DeltaSeconds :type Float%20%28single-precision%29)))

(event EventSpawnActor
  (params (New%20Transform :type Transform))
  ;[1 1 1] Comment
  (bind _returnvalue (CallFunction|GetActorClasstoSpawn))
  (if (Utilities|IsValidClass _returnvalue)
    (Utilities|Casting|CastToB_SMA_Athena_RealitySapling_BaseClass _returnvalue
      (:then
        (bind _returnvalue_2 (Game|SpawnActorfromClass "/RealitySeedGameplay/items/RealitySaplings/B_SMA_Athena_RealitySapling_Base.B_SMA_Athena_RealitySapling_Base_C" New%20Transform "Undefined" "MultiplyWithRoot" 0 (Variables|Default|GetNum%20Times%20Weeded) (Variables|Default|GetOverrride%20SaplingPersistence) (Variables|Default|GetSaplingNumTimesWateredThisMatch) (Variables|Default|GetDay%20Offset) (Variables|Default|GetFruit%201State) (Variables|Default|GetFruit%202State) (Variables|Default|GetFruit%203State) (Variables|Default|GetMovedSaplingPersistentData) (Utilities|IsValid (Utilities|Struct|BreakRealitySaplingPersistentData (Variables|Default|GetMovedSaplingPersistentData)) :!class /Script/Engine.KismetSystemLibrary) (Variables|Default|GetPlanterPlayerController) (Utilities|Struct|MakeSaplingBPReppedData (Variables|Default|GetSapling%20Age)) (Variables|Default|GetOverrideIsLocalPlayerSapling) 0 0 (Variables|Actor|GetInstigator)))
        (Default|CallEventServerActorSpawned _returnvalue_2))
      (:CastFailed
        (|CallParentFunction New%20Transform)))))

(event Custom|Event_ServerActorSpawned_Event
  (params (SpawnedActor :type Actor%20Object%20Reference))
  ;[1 1 1] Comment
  ;;[1 0.5091 0.2551] Event dispatch for spawning the actor.
  (CallFunction|ServerActorSpawned SpawnedActor))

(event Game|Damage|EventPointDamage
  (params (Damage :type Float%20%28single-precision%29) (DamageType :type Damage%20Type%20Object%20Reference) (HitLocation :type Vector :default "0, 0, 0") (HitNormal :type Vector :default "0, 0, 0") (HitComponent :type Primitive%20Component%20Object%20Reference) (BoneName :type Name :default "None") (ShotFromDirection :type Vector :default "0, 0, 0") (InstigatedBy :type Controller%20Object%20Reference) (DamageCauser :type Actor%20Object%20Reference) (HitInfo :type Hit%20Result%20Structure%20%28by%20ref%29))
  (Development|PrintString (Utilities|String|Append "POINT DAMAGE: " (Utilities|String|ToString(Float) Damage) "   :    " (Utilities|GetDisplayName DamageType))))

(event Game|Damage|EventRadialDamage
  (params (DamageReceived :type Float%20%28single-precision%29) (DamageType :type Damage%20Type%20Object%20Reference) (Origin :type Vector :default "0, 0, 0") (HitInfo :type Hit%20Result%20Structure%20%28by%20ref%29) (InstigatedBy :type Controller%20Object%20Reference) (DamageCauser :type Actor%20Object%20Reference))
  (Development|PrintString "RADIAL DAMAGE")
  (Development|PrintString (Utilities|String|Append "RADIAL DAMAGE: " (Utilities|String|ToString(Float) DamageReceived) "   :    " (Utilities|GetDisplayName DamageType))))

(event Game|Damage|EventAnyDamage
  (params (Damage :type Float%20%28single-precision%29) (DamageType :type Damage%20Type%20Object%20Reference) (InstigatedBy :type Controller%20Object%20Reference) (DamageCauser :type Actor%20Object%20Reference))
  (Development|PrintString "ANY DAMAGE")
  (Development|PrintString (Utilities|String|Append "RADIAL DAMAGE: " (Utilities|String|ToString(Float) Damage) "   :    " (Utilities|GetDisplayName DamageType))))

(fn ConstructionScript
  (|CallParentFunction))
