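; Base Reality Sapling Blueprint that plays visual and audio feedback for planting, nurturing, transplanting, ghost conversion and end of life.
; Construction script creates dynamic material instances for the young and adult FX meshes.
; Events branch on age variant and ghost state, swapping masked materials while timelines drive "SaplingTeleport" dissolves, reversed for removal.
; Locally owned saplings add Niagara effects and ambient sounds; nurturing timelines animate wiggle, petal opening and emissive.
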(event EventOnEndOfLife
  (bind _babystaticmesh (Variables|Default|GetBabyStaticMesh))
  (bind _youthstaticmesh (Variables|Default|GetYouthStaticMesh))
  (bind _staticmeshcomponent (Variables|Debug|GetStaticMeshComponent))
  (bind _stumpstaticmesh (Variables|Default|GetStumpStaticMesh))
  (bind _ns_localplayer_adult (Variables|Default|GetNS_LocalPlayer_Adult))
  ;[1 1 1] Sapling Death Event
  (|CallParentFunction)
  (if (not (Variables|Default|GetIsGhost))
    (Audio|Components|Audio|FadeOut (Variables|Default|GetAmbientLoop) 0.2 0.0 "Linear" :!class /Script/Engine.AudioComponent)
    (switch Utilities|FlowControl|Switch|SwitchonSaplingAgeVariants (CallFunction|GetSaplingAgeVariant)
      (:NewEnumerator0
        (Rendering|Material|SetMaterial _babystaticmesh 0 "/RealitySeedGameplay/Environment/Assets_RealitySapling/Materials/MI_RealitySapling_01_Masked.MI_RealitySapling_01_Masked")
        (bind (_direction _teleportcurve) (Timeline|Timeline_6 ReverseFromEnd)
          (:Update
            (Rendering|Material|SetScalarParameterValueonMaterials _babystaticmesh "SaplingTeleport" _teleportcurve))
          (:Finished
            (Components|DestroyComponent (Variables|Default|GetNS_LocalPlayer_Adult)))))
      (:NewEnumerator1
        (Rendering|Material|SetMaterial _youthstaticmesh 0 "/RealitySeedGameplay/Environment/Assets_RealitySapling/Materials/MI_RealitySapling_02_Masked.MI_RealitySapling_02_Masked")
        (bind (_direction_2 _teleportcurve_2) (Timeline|Timeline_5 ReverseFromEnd)
          (:Update
            (Rendering|Material|SetScalarParameterValueonMaterials _youthstaticmesh "SaplingTeleport" _teleportcurve_2)
            (Rendering|Material|SetScalarParameterValue (Variables|Default|GetFX_Material_Young_Fronds) "SaplingTeleport" _teleportcurve_2 :!class /Script/Engine.MaterialInstanceDynamic)
            (Rendering|Material|SetScalarParameterValue (Variables|Default|GetFX_Material_Young_GLow) "SaplingTeleport" _teleportcurve_2 :!class /Script/Engine.MaterialInstanceDynamic)
            (Rendering|Material|SetScalarParameterValue (Variables|Default|GetFX_Material_Young_GlowSmall) "SaplingTeleport" _teleportcurve_2 :!class /Script/Engine.MaterialInstanceDynamic))
          (:Finished
            (Components|DestroyComponent (Variables|Default|GetNS_LocalPlayer_Adult)))))
      (:NewEnumerator2
        _)
      (:NewEnumerator3
        _))
    (Rendering|Material|SetMaterial _staticmeshcomponent 0 "/RealitySeedGameplay/Environment/Assets_RealitySapling/Materials/MI_RealitySapling_03_Masked.MI_RealitySapling_03_Masked")
    (bind (_direction_3 _teleportcurve_3) (Timeline|Timeline_4 ReverseFromEnd)
      (:Update
        (Rendering|Material|SetScalarParameterValueonMaterials _staticmeshcomponent "SaplingTeleport" _teleportcurve_3)
        (Rendering|Material|SetScalarParameterValue (Variables|Default|GetFX_Material_Adult_Petals) "SaplingTeleport" _teleportcurve_3 :!class /Script/Engine.MaterialInstanceDynamic)
        (Rendering|Material|SetScalarParameterValue (Variables|Default|GetFX_Material_Adult_GLow) "SaplingTeleport" _teleportcurve_3 :!class /Script/Engine.MaterialInstanceDynamic)
        (Effects|Components|ParticleSystem|SetFloatParameter _ns_localplayer_adult "Opacity" _teleportcurve_3)
        (Rendering|Material|SetScalarParameterValue (Variables|Default|GetFX_Material_Adult_Spline) "SaplingTeleport" _teleportcurve_3 :!class /Script/Engine.MaterialInstanceDynamic))
      (:Finished
        (Components|DestroyComponent _ns_localplayer_adult)))
    (else
      (Rendering|Material|SetMaterial _stumpstaticmesh 0 "/RealitySeedGameplay/Environment/Assets_RealitySapling/Materials/MI_RealitySapling_04_Ghost.MI_RealitySapling_04_Ghost")
      (bind (_direction_4 _teleportcurve_4) (Timeline|Timeline_3 ReverseFromEnd)
        (:Update
          (Rendering|Material|SetScalarParameterValueonMaterials _stumpstaticmesh "SaplingTeleport" _teleportcurve_4))))))

(event Custom|OnSaplingTransplanted
  (bind _staticmeshcomponent (Variables|Debug|GetStaticMeshComponent))
  (bind _returnvalue (Transformation|GetWorldLocation _staticmeshcomponent))
  (bind _saplingage (CallFunction|GetSaplingAgeVariant))
  (bind _babystaticmesh (Variables|Default|GetBabyStaticMesh))
  (bind _youthstaticmesh (Variables|Default|GetYouthStaticMesh))
  (bind _staticmeshcomponent_2 (Variables|Debug|GetStaticMeshComponent))
  (bind _stumpstaticmesh (Variables|Default|GetStumpStaticMesh))
  ;[1 1 1] Saplinjg Teleport
  (Audio|SpawnSoundatLocation (Variables|Default|GetTransplantedSound) _returnvalue)
  (bind _returnvalue_2 (Niagara|SpawnSystematLocation "/RealitySeedGameplay/Environment/Assets_RealitySapling/FX/NS_RealitySapling_Teleport.NS_RealitySapling_Teleport" _returnvalue))
  (Niagara|SetNiagaraVariable(Int32) _returnvalue_2 "SaplingAge" (Math|Conversions|ToInteger(Byte) _saplingage))
  (if (not (Variables|Default|GetIsGhost))
    (switch Utilities|FlowControl|Switch|SwitchonSaplingAgeVariants _saplingage
      (:NewEnumerator0
        (Rendering|Material|SetMaterial _babystaticmesh 0 "/RealitySeedGameplay/Environment/Assets_RealitySapling/Materials/MI_RealitySapling_01_Masked.MI_RealitySapling_01_Masked")
        (bind (_direction _teleportcurve) (Timeline|SaplingTeleportCurve PlayFromStart)
          (:Update
            (Rendering|Material|SetScalarParameterValueonMaterials _babystaticmesh "SaplingTeleport" _teleportcurve))
          (:Finished
            (Rendering|Material|SetMaterial _babystaticmesh 0 "/RealitySeedGameplay/Environment/Assets_RealitySapling/Materials/MI_RealitySapling_01.MI_RealitySapling_01"))))
      (:NewEnumerator1
        (Rendering|Material|SetMaterial _youthstaticmesh 0 "/RealitySeedGameplay/Environment/Assets_RealitySapling/Materials/MI_RealitySapling_02_Masked.MI_RealitySapling_02_Masked")
        (bind (_direction_2 _teleportcurve_2) (Timeline|Timeline PlayFromStart)
          (:Update
            (Rendering|Material|SetScalarParameterValueonMaterials _youthstaticmesh "SaplingTeleport" _teleportcurve_2)
            (Rendering|Material|SetScalarParameterValue (Variables|Default|GetFX_Material_Young_Fronds) "SaplingTeleport" _teleportcurve_2 :!class /Script/Engine.MaterialInstanceDynamic)
            (Rendering|Material|SetScalarParameterValue (Variables|Default|GetFX_Material_Young_GLow) "SaplingTeleport" _teleportcurve_2 :!class /Script/Engine.MaterialInstanceDynamic)
            (Rendering|Material|SetScalarParameterValue (Variables|Default|GetFX_Material_Young_GlowSmall) "SaplingTeleport" _teleportcurve_2 :!class /Script/Engine.MaterialInstanceDynamic))
          (:Finished
            (Rendering|Material|SetMaterial _youthstaticmesh 0 "/RealitySeedGameplay/Environment/Assets_RealitySapling/Materials/MI_RealitySapling_02.MI_RealitySapling_02"))))
      (:NewEnumerator2
        _)
      (:NewEnumerator3
        _))
    (Rendering|Material|SetMaterial _staticmeshcomponent_2 0 "/RealitySeedGameplay/Environment/Assets_RealitySapling/Materials/MI_RealitySapling_03_Masked.MI_RealitySapling_03_Masked")
    (bind (_direction_3 _teleportcurve_3) (Timeline|Timeline_1 PlayFromStart)
      (:Update
        (Rendering|Material|SetScalarParameterValueonMaterials _staticmeshcomponent_2 "SaplingTeleport" _teleportcurve_3)
        (Rendering|Material|SetScalarParameterValue (Variables|Default|GetFX_Material_Adult_Petals) "SaplingTeleport" _teleportcurve_3 :!class /Script/Engine.MaterialInstanceDynamic)
        (Rendering|Material|SetScalarParameterValue (Variables|Default|GetFX_Material_Adult_GLow) "SaplingTeleport" _teleportcurve_3 :!class /Script/Engine.MaterialInstanceDynamic)
        (Rendering|Material|SetScalarParameterValue (Variables|Default|GetFX_Material_Adult_Spline) "SaplingTeleport" _teleportcurve_3 :!class /Script/Engine.MaterialInstanceDynamic))
      (:Finished
        (Rendering|Material|SetMaterial _staticmeshcomponent_2 0 "/RealitySeedGameplay/Environment/Assets_RealitySapling/Materials/MI_RealitySapling_03.MI_RealitySapling_03")))
    (else
      (Rendering|Material|SetMaterial _stumpstaticmesh 0 "/RealitySeedGameplay/Environment/Assets_RealitySapling/Materials/MI_RealitySapling_04_Ghost.MI_RealitySapling_04_Ghost")
      (Timeline|Timeline_2 PlayFromStart
        (:Update
          (Rendering|Material|SetScalarParameterValueonMaterials _stumpstaticmesh "SaplingTeleport" 0.0))
        (:Finished
          (Rendering|Material|SetMaterial _stumpstaticmesh 0 "/RealitySeedGameplay/Environment/Assets_RealitySapling/Materials/MI_RealitySapling_04.MI_RealitySapling_04"))))))

(event EventBeginPlay
  (|CallParentFunction)
  (if (not (Variables|Default|GetIsGhost))
    (switch Utilities|FlowControl|Switch|SwitchonSaplingAgeVariants (CallFunction|GetSaplingAgeVariant)
      (:NewEnumerator1
        (Rendering|SetVisibility (Variables|Default|GetFX_SM_Young) true false))
      (:NewEnumerator2
        _)
      (:NewEnumerator3
        _))
    (Rendering|SetVisibility (Variables|Default|GetFX_SM_Adult) true false)))

(event EventOnSaplingBelongstoLocalPlayer
  (bind _staticmeshcomponent (Variables|Debug|GetStaticMeshComponent))
  (bind _returnvalue (Transformation|GetWorldLocation _staticmeshcomponent))
  (bind _ambientloop (Variables|Default|GetAmbientLoop))
  ;[1 1 1] Local Player Sapling Only
  (|CallParentFunction)
  (if (not (Variables|Default|GetIsGhost))
    (switch Utilities|FlowControl|Switch|SwitchonSaplingAgeVariants (CallFunction|GetSaplingAgeVariant)
      (:NewEnumerator0
        (bind _returnvalue_2 (Niagara|SpawnSystematLocation "/RealitySeedGameplay/Environment/Assets_RealitySapling/FX/NS_RealitySapling_baby.NS_RealitySapling_baby" _returnvalue))
        (Variables|Default|SetNS_LocalPlayer_Adult _returnvalue_2))
      (:NewEnumerator1
        (bind _returnvalue_3 (Niagara|SpawnSystematLocation "/RealitySeedGameplay/Environment/Assets_RealitySapling/FX/NS_RealitySapling_Young.NS_RealitySapling_Young" _returnvalue))
        (Variables|Default|SetNS_LocalPlayer_Adult _returnvalue_3))
      (:NewEnumerator2
        (bind _returnvalue_4 (Niagara|SpawnSystematLocation "/RealitySeedGameplay/Environment/Assets_RealitySapling/FX/NS_RealitySapling_Flower.NS_RealitySapling_Flower" _returnvalue :!label _label))
        (Variables|Default|SetNS_LocalPlayer_Adult _returnvalue_4)
        (Audio|Components|Audio|SetSound _ambientloop (Variables|Default|GetOwnedAdultSaplingSound)))
      (:NewEnumerator3
        (goto _label)))
    (Audio|Components|Audio|SetSound _ambientloop (Variables|Default|GetOwnedBabySaplingSound))))

(event EventOnSaplingNurtured
  (bind _fx_material_adult_petals (Variables|Default|GetFX_Material_Adult_Petals))
  (bind _fx_material_adult_spline (Variables|Default|GetFX_Material_Adult_Spline))
  ;[1 1 1] Add FX and Audio for happy tree/wiggle, etc. here.
  (|CallParentFunction)
  (Development|PrintString "Sapling Nurtured")
  (bind (_direction _petalsopenming _wpo_wiggle _emissivepan) (Timeline|PetalANimation PlayFromStart)
    (:Update
      (Rendering|Material|SetScalarParameterValueonMaterials (Variables|Default|GetYouthStaticMesh) "WiggleControl" _wpo_wiggle)
      (Rendering|Material|SetScalarParameterValueonMaterials (Variables|Debug|GetStaticMeshComponent) "ReactiveEmissive" _emissivepan)
      (Rendering|Material|SetScalarParameterValue _fx_material_adult_petals "WPO Amount Opening" (* _petalsopenming 4.0) :!class /Script/Engine.MaterialInstanceDynamic)
      (Rendering|Material|SetScalarParameterValue _fx_material_adult_petals "Intensity" (* _petalsopenming 3.0) :!class /Script/Engine.MaterialInstanceDynamic)
      (Rendering|Material|SetScalarParameterValue _fx_material_adult_spline "Intensity" (* _petalsopenming 3.0) :!class /Script/Engine.MaterialInstanceDynamic)
      (Rendering|Material|SetScalarParameterValue _fx_material_adult_spline "WPO Amount Opening" (* _petalsopenming 4.0) :!class /Script/Engine.MaterialInstanceDynamic)
      (Effects|Components|ParticleSystem|SetFloatParameter (Variables|Default|GetNS_LocalPlayer_Adult) "Nurtuted" _petalsopenming))))

(event EventOnSaplingPartiallyNurtured
  (bind _fx_material_adult_petals (Variables|Default|GetFX_Material_Adult_Petals))
  (bind _fx_material_adult_spline (Variables|Default|GetFX_Material_Adult_Spline))
  ;[1 1 1] Add SMALLER FX and Audio for happy tree/wiggle, etc. here.
  (Development|PrintString "Sapling partially nurtured")
  (|CallParentFunction)
  (Utilities|FlowControl|Sequence
    (:then_0
      (bind (_direction _petalsopenming _wpo_wiggle _emissivepan) (Timeline|Timeline_0 PlayFromStart)
        (:Update
          (Rendering|Material|SetScalarParameterValueonMaterials (Variables|Default|GetBabyStaticMesh) "WiggleControl" _wpo_wiggle)
          (Rendering|Material|SetScalarParameterValueonMaterials (Variables|Debug|GetStaticMeshComponent) "ReactiveEmissive" _emissivepan)
          (Rendering|Material|SetScalarParameterValue _fx_material_adult_petals "WPO Amount Opening" (* _petalsopenming 2.5) :!class /Script/Engine.MaterialInstanceDynamic)
          (Rendering|Material|SetScalarParameterValue _fx_material_adult_petals "Intensity" (* _petalsopenming 1.0) :!class /Script/Engine.MaterialInstanceDynamic)
          (Rendering|Material|SetScalarParameterValue _fx_material_adult_spline "Intensity" (* _petalsopenming 1.0) :!class /Script/Engine.MaterialInstanceDynamic)
          (Rendering|Material|SetScalarParameterValue _fx_material_adult_spline "WPO Amount Opening" (* _petalsopenming 2.5) :!class /Script/Engine.MaterialInstanceDynamic)
          (Effects|Components|ParticleSystem|SetFloatParameter (Variables|Default|GetNS_LocalPlayer_Adult) "Nurtuted" (* _petalsopenming 0.5)))))
    (:then_1
      (Networking|SwitchHasAuthority
        (:Remote
          (Audio|SpawnSoundAttached :Sound (Variables|Default|GetNourishedSound) :AttachToComponent (Variables|Debug|GetStaticMeshComponent) :AttachPointName "None" :LocationType "SnapToTarget" :bStopWhenAttachedToDestroyed false :VolumeMultiplier 1.0 :PitchMultiplier 1.0 :StartTime 0.0 :bAutoDestroy true))))))

(event EventOnSaplingBabyPlanted
  (bind _babystaticmesh (Variables|Default|GetBabyStaticMesh))
  (bind _returnvalue (Transformation|GetWorldLocation _babystaticmesh))
  ;[1 1 1] When Sapling is Planted
  (|CallParentFunction)
  (Utilities|FlowControl|Sequence
    (:then_0
      (Niagara|SpawnSystematLocation "/RealitySeedGameplay/Environment/Assets_RealitySapling/FX/NS_RealitySapling_Planting.NS_RealitySapling_Planting" _returnvalue)
      (Development|PrintString (Utilities|String|ToString(Vector) _returnvalue) (Variables|Debug|GetDebug) (Variables|Debug|GetDebug) "(R=1.000000,G=0.000000,B=0.854337,A=1.000000)" 1e+01))
    (:then_1
      (bind (_direction _scale) (Timeline|Scale PlayFromStart)
        (:Update
          (Transformation|SetRelativeScale3D (Variables|Default|GetBabyStaticMesh) (Math|Conversions|ToVector(Float) _scale)))))))

(event EventOnTurnedIntoAGhost
  (bind _staticmeshcomponent (Variables|Debug|GetStaticMeshComponent))
  (bind _returnvalue (Transformation|GetWorldLocation _staticmeshcomponent))
  (bind _fx_sm_adult (Variables|Default|GetFX_SM_Adult))
  ;[1 1 1] When Turning Into Ghost
  (|CallParentFunction)
  (Utilities|FlowControl|Sequence
    (:then_0
      (Components|DestroyComponent (Variables|Default|GetNS_LocalPlayer_Adult))
      (if (not (Variables|Default|GetIsGhost))
        (switch Utilities|FlowControl|Switch|SwitchonSaplingAgeVariants (CallFunction|GetSaplingAgeVariant)
          (:NewEnumerator1
            (Rendering|SetVisibility (Variables|Default|GetFX_SM_Young) false false))
          (:NewEnumerator2
            _)
          (:NewEnumerator3
            _))
        (Rendering|SetVisibility _fx_sm_adult false false)
        (else
          (Rendering|SetVisibility _fx_sm_adult false false))))
    (:then_1
      (Niagara|SpawnSystematLocation "/RealitySeedGameplay/Environment/Assets_RealitySapling/FX/NS_RealitySapling_TurnGhost.NS_RealitySapling_TurnGhost" _returnvalue)
      (Audio|SpawnSoundatLocation (Variables|Default|GetDestroyedSound) _returnvalue)
      (Audio|Components|Audio|SetSound (Variables|Default|GetAmbientLoop) (Variables|Default|GetIsGhostSound)))))

(event EventOnSaplingTransplantingAway
  (bind _babystaticmesh (Variables|Default|GetBabyStaticMesh))
  (bind _youthstaticmesh (Variables|Default|GetYouthStaticMesh))
  (bind _staticmeshcomponent (Variables|Debug|GetStaticMeshComponent))
  (bind _stumpstaticmesh (Variables|Default|GetStumpStaticMesh))
  (bind _ns_localplayer_adult (Variables|Default|GetNS_LocalPlayer_Adult))
  ;[1 1 1] On Sapling Teleport Out
  (|CallParentFunction)
  (Audio|SpawnSoundatLocation (Variables|Default|GetRemovedSound) (Transformation|GetWorldLocation (Variables|Debug|GetStaticMeshComponent)))
  (Components|DestroyComponent (Variables|Default|GetNS_LocalPlayer_Adult))
  (if (not (Variables|Default|GetIsGhost))
    (switch Utilities|FlowControl|Switch|SwitchonSaplingAgeVariants (CallFunction|GetSaplingAgeVariant)
      (:NewEnumerator0
        (Rendering|Material|SetMaterial _babystaticmesh 0 "/RealitySeedGameplay/Environment/Assets_RealitySapling/Materials/MI_RealitySapling_01_Masked.MI_RealitySapling_01_Masked")
        (bind (_direction _teleportcurve) (Timeline|Timeline_7 ReverseFromEnd)
          (:Update
            (Rendering|Material|SetScalarParameterValueonMaterials _babystaticmesh "SaplingTeleport" _teleportcurve))
          (:Finished
            (Components|DestroyComponent (Variables|Default|GetNS_LocalPlayer_Adult)))))
      (:NewEnumerator1
        (Rendering|Material|SetMaterial _youthstaticmesh 0 "/RealitySeedGameplay/Environment/Assets_RealitySapling/Materials/MI_RealitySapling_02_Masked.MI_RealitySapling_02_Masked")
        (bind (_direction_2 _teleportcurve_2) (Timeline|Timeline_8 ReverseFromEnd)
          (:Update
            (Rendering|Material|SetScalarParameterValueonMaterials _youthstaticmesh "SaplingTeleport" _teleportcurve_2)
            (Rendering|Material|SetScalarParameterValue (Variables|Default|GetFX_Material_Young_Fronds) "SaplingTeleport" _teleportcurve_2 :!class /Script/Engine.MaterialInstanceDynamic)
            (Rendering|Material|SetScalarParameterValue (Variables|Default|GetFX_Material_Young_GLow) "SaplingTeleport" _teleportcurve_2 :!class /Script/Engine.MaterialInstanceDynamic)
            (Rendering|Material|SetScalarParameterValue (Variables|Default|GetFX_Material_Young_GlowSmall) "SaplingTeleport" _teleportcurve_2 :!class /Script/Engine.MaterialInstanceDynamic))
          (:Finished
            (Components|DestroyComponent (Variables|Default|GetNS_LocalPlayer_Adult)))))
      (:NewEnumerator2
        _)
      (:NewEnumerator3
        _))
    (Rendering|Material|SetMaterial _staticmeshcomponent 0 "/RealitySeedGameplay/Environment/Assets_RealitySapling/Materials/MI_RealitySapling_03_Masked.MI_RealitySapling_03_Masked")
    (bind (_direction_3 _teleportcurve_3) (Timeline|Timeline_9 ReverseFromEnd)
      (:Update
        (Rendering|Material|SetScalarParameterValueonMaterials _staticmeshcomponent "SaplingTeleport" _teleportcurve_3)
        (Rendering|Material|SetScalarParameterValue (Variables|Default|GetFX_Material_Adult_Petals) "SaplingTeleport" _teleportcurve_3 :!class /Script/Engine.MaterialInstanceDynamic)
        (Rendering|Material|SetScalarParameterValue (Variables|Default|GetFX_Material_Adult_GLow) "SaplingTeleport" _teleportcurve_3 :!class /Script/Engine.MaterialInstanceDynamic)
        (Effects|Components|ParticleSystem|SetFloatParameter _ns_localplayer_adult "Opacity" 0.0)
        (Rendering|Material|SetScalarParameterValue (Variables|Default|GetFX_Material_Adult_Spline) "SaplingTeleport" _teleportcurve_3 :!class /Script/Engine.MaterialInstanceDynamic))
      (:Finished
        (Components|DestroyComponent _ns_localplayer_adult)))
    (else
      (Rendering|Material|SetMaterial _stumpstaticmesh 0 "/RealitySeedGameplay/Environment/Assets_RealitySapling/Materials/MI_RealitySapling_04_Ghost.MI_RealitySapling_04_Ghost")
      (Development|PrintString "StumpMeshTeleoportout" true true "(R=1.000000,G=0.000000,B=0.000000,A=1.000000)" 1e+01)
      (bind (_direction_4 _teleportcurve_4) (Timeline|Timeline_10 ReverseFromEnd)
        (:Update
          (Rendering|Material|SetScalarParameterValueonMaterials _stumpstaticmesh "SaplingTeleport" _teleportcurve_4))))))

(event EventEndPlay
  (params (EndPlayReason :type EEndPlayReason%20Enum :default "Destroyed"))
  (Components|DestroyComponent (Variables|Default|GetNS_LocalPlayer_Adult)))

(event Collision|EventActorBeginOverlap
  (params (OtherActor :type Actor%20Object%20Reference)))

(event EventTick
  (params (DeltaSeconds :type Float%20%28single-precision%29)))

(fn OnSaplingNewlyTransplanted
  (|CallParentFunction)
  (CallFunction|OnSaplingTransplanted))

(fn ConstructionScript
  (|CallParentFunction)
  (bind _returnvalue (Rendering|Material|CreateDynamicMaterialInstance (Variables|Default|GetFX_SM_Adult) 0 "/RealitySeedGameplay/Environment/Assets_RealitySapling/FX/MI_RealitySapling_Petal_Inst.MI_RealitySapling_Petal_Inst" "None" :!class /Script/Engine.PrimitiveComponent))
  (Variables|Default|SetFX_Material_Adult_Petals _returnvalue)
  (bind _returnvalue_2 (Rendering|Material|CreateDynamicMaterialInstance (Variables|Default|GetFX_SM_Adult) 1 "/RealitySeedGameplay/Environment/Assets_RealitySapling/FX/M_RealitySapling_Glow_Inst.M_RealitySapling_Glow_Inst" "None" :!class /Script/Engine.PrimitiveComponent))
  (Variables|Default|SetFX_Material_Adult_GLow _returnvalue_2)
  (bind _returnvalue_3 (Rendering|Material|CreateDynamicMaterialInstance (Variables|Default|GetFX_SM_Young) 0 "/RealitySeedGameplay/Environment/Assets_RealitySapling/FX/M_RealitySapling_Glow.M_RealitySapling_Glow" "None" :!class /Script/Engine.PrimitiveComponent))
  (Variables|Default|SetFX_Material_Young_Fronds _returnvalue_3)
  (bind _returnvalue_4 (Rendering|Material|CreateDynamicMaterialInstance (Variables|Default|GetFX_SM_Young) 2 "/RealitySeedGameplay/Environment/Assets_RealitySapling/FX/M_RealitySapling_Wind_Inst.M_RealitySapling_Wind_Inst" "None" :!class /Script/Engine.PrimitiveComponent))
  (Variables|Default|SetFX_Material_Young_GLow _returnvalue_4)
  (bind _returnvalue_5 (Rendering|Material|CreateDynamicMaterialInstance (Variables|Default|GetFX_SM_Young) 1 "/RealitySeedGameplay/Environment/Assets_RealitySapling/FX/M_RealitySapling_GlowSprite.M_RealitySapling_GlowSprite" "None" :!class /Script/Engine.PrimitiveComponent))
  (Variables|Default|SetFX_Material_Young_GlowSmall _returnvalue_5)
  (bind _returnvalue_6 (Rendering|Material|CreateDynamicMaterialInstance (Variables|Default|GetFX_SM_Adult2) 0 "/RealitySeedGameplay/Environment/Assets_RealitySapling/FX/M_RealitySapling_Spline_02.M_RealitySapling_Spline_02" "None" :!class /Script/Engine.PrimitiveComponent))
  (Variables|Default|SetFX_Material_Adult_Spline _returnvalue_6))
