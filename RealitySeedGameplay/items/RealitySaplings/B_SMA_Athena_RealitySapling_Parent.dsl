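; Parent Blueprint for the Reality Sapling, a player-planted, nurturable tree that persists between matches and grows harvestable fruit.
; On BeginPlay the server copies persistent data (owner, nurturing, maturity, fruit states) into replicated variables, handling transplanted saplings.
; Initialization sets the age or ghost mesh, spawns fruit, weeds and mushrooms, and withers the sapling once its mythic fruit is picked.
; It also auto-pings on clients and sends cooldown-limited alerts when enemies damage or harvest it.
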(event Custom|BindToBoxOverlapEvent
  (params (Begin :type Boolean))
  ;[1 1 1] Bind to/End box overlap events
  (Networking|SwitchHasAuthority
    (:Authority
      (if Begin
        (Collision|BindEventtoOnComponentBeginOverlap (Variables|Default|GetBox) (delegate Custom|StartBoxOverlap))
        (else
          (Collision|UnbindEventfromOnComponentBeginOverlap (Variables|Default|GetBox) (delegate Custom|StartBoxOverlap)))))))

(event Custom|Water%20With%20Gasoline)

(event Custom|StartBoxOverlap
  (params (OverlappedComponent :type Primitive%20Component%20Object%20Reference) (OtherActor :type Actor%20Object%20Reference) (OtherComp :type Primitive%20Component%20Object%20Reference) (OtherBodyIndex :type Integer) (bFromSweep :type Boolean) (SweepResult :type Hit%20Result%20Structure%20%28by%20ref%29)))

(event Custom|FillVariablesFromPersistentData
  (bind _minnurturinglevel (Variables|Default|GetMinNurturingLevel))
  (bind _returnvalue (ScalableFloat|GetValueasInteger _minnurturinglevel))
  (bind (_saplingbpreppeddata_saplingage_4_5368fd2741dc1813a345149d36ca3e98 _saplingbpreppeddata_revivecount_5_ec5af65440fb691c7b8724bd7cdfa3e1) (Utilities|Struct|BreakSaplingBPReppedData (Variables|GetSaplingBPReppedData_SaplingAge_4_5368FD2741DC1813A345149D36CA3E98)))
  (bind _saplingbpreppeddata_saplingage_4_5368fd2741dc1813a345149d36ca3e98_2 (Utilities|Struct|BreakSaplingBPReppedData (Variables|GetSaplingBPReppedData_SaplingAge_4_5368FD2741DC1813A345149D36CA3E98)))
  (bind (_persistentsaplingdata_owningplayercontroller _persistentsaplingdata_location _persistentsaplingdata_rotation _persistentsaplingdata_nurturinglevel _persistentsaplingdata_maturitylevel _persistentsaplingdata_plantingtime _persistentsaplingdata_lastweedingtime) (Utilities|Struct|BreakRealitySaplingPersistentData (Variables|GetPersistentSaplingData_OwningPlayerController)))
  ;[0.6214 1 0.6973] Server - Copy persisetnt sapling data into BP used variables (some which are replicated)
  (Networking|SwitchHasAuthority
    (:Authority
      (if (not (Variables|Default|GetOverrride%20Sapling%20Persistence))
        (if (<= _persistentsaplingdata_nurturinglevel _returnvalue)
          (Variables|Default|SetSaplingNurturedCountThisMatch _returnvalue)
          (else
            (Variables|Default|SetSaplingNurturedCountThisMatch _persistentsaplingdata_nurturinglevel)))
        (Variables|Persistent|SetLastWeededDateTime _persistentsaplingdata_lastweedingtime)
        (if (>= _persistentsaplingdata_nurturinglevel (ScalableFloat|GetValueasInteger (Variables|Default|GetMaxWeedingTimes)))
          (Variables|SetSaplingBPReppedData_SaplingAge_4_5368FD2741DC1813A345149D36CA3E98 (Utilities|Struct|MakeSaplingBPReppedData "NewEnumerator3" _saplingbpreppeddata_revivecount_5_ec5af65440fb691c7b8724bd7cdfa3e1))
          (else
            (switch int _persistentsaplingdata_maturitylevel
              (:Default
                (Variables|SetSaplingBPReppedData_SaplingAge_4_5368FD2741DC1813A345149D36CA3E98 (Utilities|Struct|MakeSaplingBPReppedData "NewEnumerator2" _saplingbpreppeddata_revivecount_5_ec5af65440fb691c7b8724bd7cdfa3e1)))
              (:0
                (Variables|SetSaplingBPReppedData_SaplingAge_4_5368FD2741DC1813A345149D36CA3E98 (Utilities|Struct|MakeSaplingBPReppedData "NewEnumerator0" _saplingbpreppeddata_revivecount_5_ec5af65440fb691c7b8724bd7cdfa3e1)))
              (:1
                (Variables|SetSaplingBPReppedData_SaplingAge_4_5368FD2741DC1813A345149D36CA3E98 (Utilities|Struct|MakeSaplingBPReppedData "NewEnumerator1" _saplingbpreppeddata_revivecount_5_ec5af65440fb691c7b8724bd7cdfa3e1))))))
        (Development|PrintString (Utilities|String|EnumtoString _saplingbpreppeddata_saplingage_4_5368fd2741dc1813a345149d36ca3e98_2) (Variables|Debug|GetDebug) (Variables|Debug|GetDebug))
        (else
          (Variables|Default|SetSaplingBPReppedData (Variables|Default|GetSaplingBPReppedData)))))))

(event EventBeginPlay
  (bind (_owningplayercontroller _location _rotation _nurturinglevel _maturitylevel _plantingtime _lastweedingtime _fruitstates) (Utilities|Struct|BreakRealitySaplingPersistentData (Variables|RealitySapling|GetPersistentSaplingData)))
  (bind _minnurturinglevel (Variables|Default|GetMinNurturingLevel))
  (bind _returnvalue (ScalableFloat|GetValueasInteger _minnurturinglevel))
  (bind _numfruitstogrow (Variables|Default|GetNumFruitsToGrow))
  (bind _returnvalue_2 (ScalableFloat|GetValueasInteger _numfruitstogrow))
  (bind _planterplayercontroller (Variables|Default|GetPlanterPlayerController))
  (bind _persistentsaplingdata_owningplayercontroller (Utilities|Struct|BreakRealitySaplingPersistentData (Variables|GetPersistentSaplingData_OwningPlayerController)))
  (bind _serverspawnedfruitslot (Variables|Default|GetServerSpawnedFruitSlot))
  ;[0.7227 1 0.6475] BeginPlay
  (Utilities|FlowControl|Sequence
    (:then_0
      ;;[0.7126 0.8802 1] On the server, update PersistentSaplingData with variables set on spawn if necessary (This is used when the sapling is spawned from a seed)
      (Networking|SwitchHasAuthority
        (:Authority
          ;;;[1 1 1] Add empty fruits slots
          (if (>= _returnvalue_2 1)
            (Utilities|Array|Add _serverspawnedfruitslot (struct :Slot_4_56BABD8945FE8B0C64A28CBC0719D0A6 0 :Fruit_5_DCCA3B6B43057342B19A728973ABBF7E 0))
            (if (>= _returnvalue_2 2)
              (Utilities|Array|Add _serverspawnedfruitslot (struct :Slot_4_56BABD8945FE8B0C64A28CBC0719D0A6 1 :Fruit_5_DCCA3B6B43057342B19A728973ABBF7E 0))
              (if (>= _returnvalue_2 3)
                (Utilities|Array|Add _serverspawnedfruitslot (struct :Slot_4_56BABD8945FE8B0C64A28CBC0719D0A6 2 :Fruit_5_DCCA3B6B43057342B19A728973ABBF7E 0)))))
          (if (Variables|Default|GetSaplingTransplanted)
            ;;;[0.05286 0.2307 0.4072] If the sapling is transplanted, set the PersistentSaplingData from the
            ;;; MovedSaplingPersistentData
            (Variables|RealitySapling|SetPersistentSaplingData (Variables|Default|GetMovedSaplingPersistentData))
            (Development|PrintString "Sapling Transplanted, PersistentSaplingData overwritten by MovedSaplingPersistentData" (Variables|Debug|GetDebug) (Variables|Debug|GetDebug)))
          ;;;[0.1309 0.5737 1] If a PlanterPlayerController was specified (if a player planted this sapling), set that as the SaplingPersistentData OwningPlayerController (OwningPlayerControllerRepped/OwningPlayerStateRepped will later be set from this)
          (Utilities|IsValid _planterplayercontroller
            (:Is%20Valid
              (Variables|RealitySapling|SetPersistentSaplingData (Utilities|Struct|MakeRealitySaplingPersistentData _planterplayercontroller _location _rotation _nurturinglevel _maturitylevel _plantingtime _lastweedingtime _fruitstates)))
            (:Is%20Not%20Valid
              _))
          ;;[0.2575 0.3538 1] Save off the current value of the persistent sapling data as PersistentSaplingDataWhenSpawned. This will be used later if the sapling is transplanted
          (Variables|Persistent|SetPersistentDataWhenSpawned (Variables|RealitySapling|GetPersistentSaplingData)))))
    (:then_1
      (|CallParentFunction) ; Allow cvars to override persistent data in code
      (Audio|Components|Audio|Play (Variables|Default|GetAmbientLoop) 0.0)
      (if (not (Editor|InEditororPIE))
        (Variables|Debug|SetbUseShortTimes false)
        (if (not (Variables|Default|GetOverrride%20Sapling%20Persistence))
          (Variables|Debug|SetDayOffset 0)))
      (Networking|SwitchHasAuthority
        (:Authority
          (Development|PrintString (Utilities|String|Append "Persistent Sapling Controller is: " (Utilities|GetDisplayName _persistentsaplingdata_owningplayercontroller)) (Variables|Debug|GetDebug) (Variables|Debug|GetDebug))
          (Variables|Default|SetOwning%20Player%20Controller%20Repped _persistentsaplingdata_owningplayercontroller)
          (CallFunction|FillVariablesfromPersistentData) ; Fill BP variables from the persistent data struct,\nincluding setting the saplings age
          (Utilities|IsValid _persistentsaplingdata_owningplayercontroller
            (:Is%20Valid
              (bind _asfort_player_state_athena (Utilities|Casting|CastToFortPlayerStateAthena (Variables|/Script/Engine.Controller|GetPlayerState _persistentsaplingdata_owningplayercontroller))
                (:then
                  (Variables|Default|SetOwning%20Player%20State%20Repped _asfort_player_state_athena)
                  (Variables|Default|SetOwningPlayerSocialAvatarIcon (Class|FortPlayerStateZone|GetSocialAvatarBrush _asfort_player_state_athena true)))
                (:CastFailed
                  _)))
            (:Is%20Not%20Valid
              _)))
        (:Remote
          _))
      (Development|PrintString (Utilities|String|Append "Sapling Owner is: " (Utilities|GetDisplayName (Actor|GetOwner))) (Variables|Debug|GetDebug) (Variables|Debug|GetDebug))
      (CallFunction|BindtoBoxOverlapEvent true)
      (Utilities|FlowControl|Sequence
        (:then_0
          (Networking|SwitchHasAuthority
            (:Authority
              (Utilities|FlowControl|Sequence
                (:then_0
                  (if (Variables|Default|GetOverrride%20Sapling%20Persistence)
                    (if (<= (Variables|Persistent|GetNumTimesWeeded) _returnvalue)
                      (Variables|Default|SetSaplingNurturedCountThisMatch _returnvalue)
                      (else
                        (Variables|Default|SetSaplingNurturedCountThisMatch (Variables|Persistent|GetNumTimesWeeded))))
                    (Variables|Persistent|SetLastWeededDateTime (CallFunction|GetEventManagerCalendarTime))))))))))
    (:then_2
      (Variables|Default|SetHasBegunPlay true)
      (if (Variables|Default|GetHasHeldOffInitialize)
        (CallFunction|InitializeSapling)))))

(event Custom|InitializeSapling
  ;[1 1 1] Initialize Sapling
  (if (Variables|Default|GetHasBegunPlay)
    (Utilities|FlowControl|Sequence
      (:then_0
        (Variables|Default|SetWeakPointComp (Actor|GetComponentbyClass "/Script/FortniteGame.FortWeakPointComponent"))
        (Variables|Default|SetWeakPointCmps (Actor|GetComponentsByClass "/Script/FortniteGame.FortWeakPointComponent"))
        (Class|FortWeakPointComponent|SetWeakPointsEnabled (Variables|Default|GetWeakPointComp) (not (CallFunction|IsDead)))
        (if (Math|Boolean|ORBoolean (and (not (ScalableFloat|GetValueasBool (Variables|Default|GetMythicFruitEnabled))) (>= (Variables|Persistent|GetNumTimesWeeded) (ScalableFloat|GetValueasInteger (Variables|Default|GetMaxWeedingTimes)))) (> (Variables|Persistent|GetNumTimesWeeded) (ScalableFloat|GetValueasInteger (Variables|Default|GetMaxWeedingTimes))) (not (ScalableFloat|GetValueasBool (Variables|Default|GetSaplingsEnabled))))
          ;;[1 1 1] Destroy if too old
          (Variables|Default|SetDestroyedFromAging true)
          (Development|PrintString "Sapling is too old.  Calling silent die.")
          (CallFunction|SetShouldEnterGhostStateonDeath false) ; Allow to fully die
          (CallFunction|KillWeeds)
          (Gameplay|SilentDie false)
          (elif (CallFunction|IsDead)
            (if (not (Variables|Default|GetFruitFallsOffSaplingsWhenKilled))
              (for _array_element (Variables|Default|GetServerSpawnedFruitSlot)
                (bind (_array_element_slot_4_56babd8945fe8b0c64a28cbc0719d0a6 _array_element_fruit_5_dcca3b6b43057342b19a728973abbf7e) (Utilities|Struct|BreakServerSpawnedFruitSlot _array_element))
                (Utilities|IsValid _array_element_fruit_5_dcca3b6b43057342b19a728973abbf7e
                  (:Is%20Valid
                    (Class|TieredRealityFruitAthena|ScaleOutAndDie _array_element_fruit_5_dcca3b6b43057342b19a728973abbf7e)
                    (Networking|ForceNetUpdate _array_element_fruit_5_dcca3b6b43057342b19a728973abbf7e)))))
            ;;[1 1 1] Ghost state\Is Dead = true
            (CallFunction|SetSaplingMeshActive (Variables|Default|GetBabyStaticMesh) false false "/RealitySeedGameplay/Environment/Assets_RealitySapling/Materials/MI_RealitySapling_01.MI_RealitySapling_01" (Variables|Default|GetDead%20Baby%20Material))
            (CallFunction|SetSaplingMeshActive (Variables|Default|GetYouthStaticMesh) false false "/RealitySeedGameplay/Environment/Assets_RealitySapling/Materials/MI_RealitySapling_02.MI_RealitySapling_02" (Variables|Default|GetDead%20Youth%20Material))
            (CallFunction|SetSaplingMeshActive (Variables|Debug|GetStaticMeshComponent) false false "/RealitySeedGameplay/Environment/Assets_RealitySapling/Materials/MI_RealitySapling_03.MI_RealitySapling_03" (Variables|Default|GetDead%20Mature%20Material))
            (CallFunction|SetSaplingMeshActive (Variables|Default|GetStumpStaticMesh) true false "/RealitySeedGameplay/Environment/Assets_RealitySapling/Materials/MI_RealitySapling_04.MI_RealitySapling_04" (Variables|Default|GetDead%20Stump%20Material))
            (CallFunction|OnTurnedIntoAGhost)
            (else
              (Utilities|FlowControl|Sequence
                (:then_0
                  (CallFunction|CalcNeedsWeeding))
                (:then_1
                  (CallFunction|SpawnMissingFruitActors)
                  (CallFunction|SpawnWeedsifNeeded)
                  (CallFunction|SpawnMushrooms)
                  (switch Utilities|FlowControl|Switch|SwitchonSaplingAgeVariants (CallFunction|GetSaplingAgeVariant)
                    (:NewEnumerator0
                      ;;[1 1 1] Set Baby Scale
                      (Transformation|SetWorldScale3D (Variables|Default|GetBabyStaticMesh) (Utilities|Operators|Multiply (Transformation|GetWorldScale (Variables|Default|GetBabyStaticMesh)) (Math|Vector|MakeVector (Variables|Default|GetBabyScale) (Variables|Default|GetBabyScale) (Variables|Default|GetBabyScale))))
                      (CallFunction|SetSaplingMeshActive (Variables|Default|GetBabyStaticMesh) true true "/RealitySeedGameplay/Environment/Assets_RealitySapling/Materials/MI_RealitySapling_01.MI_RealitySapling_01" (Variables|Default|GetDead%20Baby%20Material))
                      (CallFunction|SetSaplingMeshActive (Variables|Default|GetYouthStaticMesh) false false "/RealitySeedGameplay/Environment/Assets_RealitySapling/Materials/MI_RealitySapling_02.MI_RealitySapling_02" (Variables|Default|GetDead%20Youth%20Material))
                      (CallFunction|SetSaplingMeshActive (Variables|Debug|GetStaticMeshComponent) false false "/RealitySeedGameplay/Environment/Assets_RealitySapling/Materials/MI_RealitySapling_03.MI_RealitySapling_03" (Variables|Default|GetDead%20Mature%20Material))
                      (CallFunction|SetSaplingMeshActive (Variables|Default|GetStumpStaticMesh) false false "/RealitySeedGameplay/Environment/Assets_RealitySapling/Materials/MI_RealitySapling_04.MI_RealitySapling_04" (Variables|Default|GetDead%20Stump%20Material))
                      (if (Variables|Default|GetSaplingTransplanted)
                        (CallFunction|CheckforSaplingNewlyTransplanted)
                        (else
                          (CallFunction|OnSaplingBabyPlanted))))
                    (:NewEnumerator1
                      ;;[1 1 1] Scale Down Youth
                      (Transformation|SetWorldScale3D (Variables|Default|GetYouthStaticMesh) (Utilities|Operators|Multiply (Transformation|GetWorldScale (Variables|Debug|GetStaticMeshComponent)) (Math|Vector|MakeVector (Variables|Default|GetYouthScale) (Variables|Default|GetYouthScale) (Variables|Default|GetYouthScale))))
                      (CallFunction|SetSaplingMeshActive (Variables|Default|GetBabyStaticMesh) false false "/RealitySeedGameplay/Environment/Assets_RealitySapling/Materials/MI_RealitySapling_01.MI_RealitySapling_01" (Variables|Default|GetDead%20Baby%20Material))
                      (CallFunction|SetSaplingMeshActive (Variables|Default|GetYouthStaticMesh) true true "/RealitySeedGameplay/Environment/Assets_RealitySapling/Materials/MI_RealitySapling_02.MI_RealitySapling_02" (Variables|Default|GetDead%20Youth%20Material))
                      (CallFunction|SetSaplingMeshActive (Variables|Debug|GetStaticMeshComponent) false false "/RealitySeedGameplay/Environment/Assets_RealitySapling/Materials/MI_RealitySapling_03.MI_RealitySapling_03" (Variables|Default|GetDead%20Mature%20Material))
                      (CallFunction|SetSaplingMeshActive (Variables|Default|GetStumpStaticMesh) false false "/RealitySeedGameplay/Environment/Assets_RealitySapling/Materials/MI_RealitySapling_04.MI_RealitySapling_04" (Variables|Default|GetDead%20Stump%20Material))
                      (CallFunction|CheckforSaplingNewlyTransplanted))
                    (:NewEnumerator2
                      _)
                    (:NewEnumerator3
                      _))
                  ;;[1 1 1] Mature
                  (CallFunction|SetSaplingMeshActive (Variables|Default|GetBabyStaticMesh) false false "/RealitySeedGameplay/Environment/Assets_RealitySapling/Materials/MI_RealitySapling_01.MI_RealitySapling_01" (Variables|Default|GetDead%20Baby%20Material))
                  (CallFunction|SetSaplingMeshActive (Variables|Default|GetYouthStaticMesh) false false "/RealitySeedGameplay/Environment/Assets_RealitySapling/Materials/MI_RealitySapling_02.MI_RealitySapling_02" (Variables|Default|GetDead%20Youth%20Material))
                  (CallFunction|SetSaplingMeshActive (Variables|Debug|GetStaticMeshComponent) true true "/RealitySeedGameplay/Environment/Assets_RealitySapling/Materials/MI_RealitySapling_03.MI_RealitySapling_03" (Variables|Default|GetDead%20Mature%20Material))
                  (CallFunction|SetSaplingMeshActive (Variables|Default|GetStumpStaticMesh) false false "/RealitySeedGameplay/Environment/Assets_RealitySapling/Materials/MI_RealitySapling_04.MI_RealitySapling_04" (Variables|Default|GetDead%20Stump%20Material))
                  (CallFunction|CheckforSaplingNewlyTransplanted)))))))
      (:then_1
        (CallFunction|AgeUpSapling))
      (:then_2
        (Utilities|Casting|CastToFortPlayerControllerAthena (Variables|Default|GetOwning%20Player%20Controller%20Repped))
        (Networking|SwitchHasAuthority
          (:Remote
            (Utilities|Time|SetTimerbyEvent (delegate Custom|SafeZoneCheck) 5.0 true)
            (bind _returnvalue (SafeZone|IsLocationinSafeZone (Transformation|GetActorLocation)))
            (CallFunction|UpdateSafeZoneStatus _returnvalue))))
      (:then_3
        (Networking|SwitchHasAuthority
          (:Authority
            (CallFunction|UpdateControllerCompSaplingDataStruct) ; Update struct on controller component for UI purposes
            (Variables|Default|SetHasBeenInitialized true))))
      (:then_4
        ;;[1 1 1] Autoping or prepare for autoping
        (Networking|SwitchHasAuthority
          (:Remote
            (Utilities|Casting|CastToFortPlayerControllerAthena (Variables|Default|GetOwning%20Player%20Controller%20Repped))
            (bind _returnvalue_2 (GameplayEventRouter|GetEventRouter "/Script/FortniteGame.FortGameState" self))
            (bind (_asynctaskproxy _payload) (GameplayEventRouter|StartListeningToStatefulEvent _returnvalue_2 "/Script/FortniteGame.FortGamePhaseStepUpdatedEvent")
              (:then
                (Variables|Default|SetGamePhaseStepUpdatedAsyncTask _asynctaskproxy))
              (:OnEventReceived
                _)
              (:OnSavedEventStateExists
                _))
            (CallFunction|GamePhaseStepChanged (Utilities|Struct|BreakFortGamePhaseStepUpdatedEvent _payload)))))
      (:then_5
        ;;[1 1 1] Damaged event to track enemy behavior
        (Networking|SwitchHasAuthority
          (:Authority
            (EventDispatchers|UnbindEventfromOnDamaged (delegate Custom|CustomEvent) :!class /Script/FortniteGame.BuildingActor)
            (EventDispatchers|BindEventtoOnDamaged (delegate Custom|CustomEvent) :!class /Script/FortniteGame.BuildingActor)))))
    (else
      (Variables|Default|SetHasHeldOffInitialize true))))

(event Custom|SafeZoneCheck
  ;[1 1 1] Initialize Sapling
  (bind _returnvalue (SafeZone|IsLocationinSafeZone (Transformation|GetActorLocation)))
  (CallFunction|UpdateSafeZoneStatus _returnvalue))

(event Custom|GamePhaseStepChanged
  (params (GamePhaseStep :type EAthenaGamePhaseStep%20Enum))
  ;[1 1 1] Initialize Sapling
  ;;[1 1 1] Autoping or prepare for autoping
  (if (>= (Math|Conversions|ToInteger(Byte) GamePhaseStep) (Math|Conversions|ToInteger(Byte) (Utilities|Enum|LiteralenumEAthenaGamePhaseStep "StormHolding")))
    (if (CallFunction|IsOktoAutoping)
      (Default|MarkSapling true)
      (bind _gamephasestepupdatedasynctask (Variables|Default|GetGamePhaseStepUpdatedAsyncTask)
        (:then
          (GameplayEventRouter|StopListeningtoStatefulEvent _gamephasestepupdatedasynctask))
        (:else)))))

(event EventEndPlay
  (params (EndPlayReason :type EEndPlayReason%20Enum :default "Destroyed"))
  ;[1 1 1] Initialize Sapling
  ;;[1 1 1] Autoping or prepare for autoping
  (bind _gamephasestepupdatedasynctask (Variables|Default|GetGamePhaseStepUpdatedAsyncTask)
    (:then
      (GameplayEventRouter|StopListeningtoStatefulEvent _gamephasestepupdatedasynctask))
    (:else)))

(event Custom|SetNotificationCooldown
  ;[1 1 1] Notifications
  (Utilities|Time|SetTimerbyEvent (delegate Custom|NotificationCooldownExpired) (CallFunction|GetNotificationCooldown))
  (Variables|Default|SetOkToNotify false))

(event Custom|NotificationCooldownExpired
  ;[1 1 1] Notifications
  (Variables|Default|SetOkToNotify true))

(event Custom|SpawnMissingFruitActors
  (bind _numfruitstogrow (Variables|Default|GetNumFruitsToGrow))
  (bind _returnvalue (ScalableFloat|GetValueasInteger _numfruitstogrow))
  (bind _retval (CallFunction|GetNumFruitsToGrowatCurrentRarity))
  (bind (_persistentsaplingdata_owningplayercontroller _persistentsaplingdata_location _persistentsaplingdata_rotation _persistentsaplingdata_nurturinglevel _persistentsaplingdata_maturitylevel _persistentsaplingdata_plantingtime _persistentsaplingdata_lastweedingtime _persistentsaplingdata_fruitstates) (Utilities|Struct|BreakRealitySaplingPersistentData (Variables|GetPersistentSaplingData_OwningPlayerController)))
  (bind _serverspawnedfruitslot (Variables|Default|GetServerSpawnedFruitSlot))
  ;[1 1 1] Initialize Sapling
  ;;[1 1 1] Spawn Fruit Actors
  (if (CallFunction|IsDead)
    _
    (elif (ScalableFloat|GetValueasBool (Variables|Default|GetRemoveAllFruitsWhenNeedsWeeding))
      (if (Variables|Persistent|GetNeedsWeeding)
        _
        (else
          _))))
  (switch Utilities|FlowControl|Switch|SwitchonSaplingAgeVariants (CallFunction|GetSaplingAgeVariant)
    (:NewEnumerator1
      _)
    (:NewEnumerator2
      _)
    (:NewEnumerator3
      _))
  (Networking|SwitchHasAuthority
    (:Authority
      (Utilities|FlowControl|Sequence
        (:then_0
          (if (not (Variables|Default|GetHasBeenInitialized))
            (Utilities|FlowControl|Sequence
              (:then_0
                (CallFunction|InitLocalPersistentSaplingDataFruitStates))
              (:then_1
                (if (Variables|Default|GetSaplingTransplanted)
                  _
                  (elif (>= (Variables|Persistent|GetNumTimesWeeded) (ScalableFloat|GetValueasInteger (Variables|Default|GetMaxWeedingTimes)))
                    (if (not (Variables|Default|GetHasSpawnedMythic))
                      (Variables|Default|SetFoundMythic false)
                      (for (_array_element _array_index) (Variables|Default|GetPersistent%20Sapling%20Data%20Fruit%20States)
                        (if (Utilities|Enum|Equal(Enum) (Utilities|Enum|BytetoEnumSaplingFruitVariants (Math|Conversions|ToByte(Integer) _array_element)) "NewEnumerator7")
                          (bind _output_get (Variables|Default|SetInitialSaplingFruitVariant "NewEnumerator8"))
                          (Utilities|Array|SetArrayElem (Variables|Default|GetPersistent%20Sapling%20Data%20Fruit%20States) _array_index (Math|Conversions|ToInteger(Byte) _output_get))
                          (elif (Utilities|Enum|Equal(Enum) (Utilities|Enum|BytetoEnumSaplingFruitVariants (Math|Conversions|ToByte(Integer) _array_element)) "NewEnumerator8")
                            _
                            (else
                              (Variables|Default|SetInitialSaplingFruitVariant "NewEnumerator0"))))
                        (Variables|Default|SetFoundMythic true))
                      (if (not (Variables|Default|GetFoundMythic))
                        (bind _output_get_2 (Variables|Default|SetInitialSaplingFruitVariant "NewEnumerator8"))
                        (Utilities|Array|SetArrayElem (Variables|Default|GetPersistent%20Sapling%20Data%20Fruit%20States) 0 (Math|Conversions|ToInteger(Byte) _output_get_2))))
                    (else
                      (CallFunction|CheckMatchStartHotFixFruitStateOverrides)))))
              (:then_2
                (if (Variables|Debug|GetDebug)
                  (Development|PrintString (Utilities|String|Append "Persistent Fruit States: Length = " (Utilities|String|ToString(Integer) (Utilities|Array|Length _persistentsaplingdata_fruitstates)) "   [" (Utilities|String|ToString(Integer) (Utilities|Array|Get(aref) _persistentsaplingdata_fruitstates)) "][" (Utilities|String|ToString(Integer) (Utilities|Array|Get(aref) _persistentsaplingdata_fruitstates 1)) "][" (Utilities|String|ToString(Integer) (Utilities|Array|Get(aref) _persistentsaplingdata_fruitstates 2)) "]" "   vs FruitStates: Length = " (Utilities|String|ToString(Integer) (Utilities|Array|Length (Variables|Default|GetPersistent%20Sapling%20Data%20Fruit%20States))) "   [" (Utilities|String|ToString(Integer) (Utilities|Array|Get(aref) (Variables|Default|GetPersistent%20Sapling%20Data%20Fruit%20States))) "][" (Utilities|String|ToString(Integer) (Utilities|Array|Get(aref) (Variables|Default|GetPersistent%20Sapling%20Data%20Fruit%20States) 1)) "][" (Utilities|String|ToString(Integer) (Utilities|Array|Get(aref) (Variables|Default|GetPersistent%20Sapling%20Data%20Fruit%20States) 2)) "]" "    Server Spawned Fruit Len:" (Utilities|String|ToString(Integer) (Utilities|Array|Length (Variables|Default|GetServerSpawnedFruitSlot)))) (Variables|Debug|GetDebug) (Variables|Debug|GetDebug) "(R=1.000000,G=1.000000,B=1.000000,A=1.000000)" 1e+02)))
              (:then_3
                ;;;[1 1 1] Spawn Initial Fruits From Persistent Data
                (bind (_array_element_2 _array_index_2) (Utilities|Array|ForEachLoopwithBreak (Variables|Default|GetPersistent%20Sapling%20Data%20Fruit%20States) :!label _label)
                  (:LoopBody
                    (Variables|Default|SetInitialSaplingFruitVariant (Utilities|Enum|BytetoEnumSaplingFruitVariants (Math|Conversions|ToByte(Integer) _array_element_2)))
                    (bind (_output_slot_4_56babd8945fe8b0c64a28cbc0719d0a6 _output_fruit_5_dcca3b6b43057342b19a728973abbf7e) (Utilities|Struct|BreakServerSpawnedFruitSlot (Utilities|Array|Get(aref) _serverspawnedfruitslot _array_index_2)))
                    (Utilities|IsValid _output_fruit_5_dcca3b6b43057342b19a728973abbf7e
                      (:Is%20Valid
                        (Development|PrintString (Utilities|String|Append "Initial Fruit already exists in slot: " (Utilities|String|ToString(Integer) _output_slot_4_56babd8945fe8b0c64a28cbc0719d0a6)) (Variables|Debug|GetDebug) (Variables|Debug|GetDebug) "(R=1.000000,G=0.000000,B=0.091064,A=1.000000)" 2e+02))
                      (:Is%20Not%20Valid
                        (switch Utilities|FlowControl|Switch|SwitchonSaplingFruitVariants (Variables|Default|GetInitialSaplingFruitVariant)
                          (:NewEnumerator0
                            (if (>= (Variables|Persistent|GetNumTimesWeeded) (ScalableFloat|GetValueasInteger (Variables|Default|GetMaxWeedingTimes)))
                              _
                              (elif (Variables|Persistent|GetNeedsWeeding)
                                ;;;;[1 1 1] Check if it's the first weeding create min level fruits at match start.
                                (if (not (and (<= (Variables|Persistent|GetNumTimesWeeded) (ScalableFloat|GetValueasInteger (Variables|Default|GetMinNurturingLevel))) (not (Variables|Default|GetSaplingTransplanted))))
                                  (Development|PrintString "Initial Needs Weeding.  Not creating new Fruit." (Variables|Debug|GetDebug) (Variables|Debug|GetDebug)))))
                            (Variables|Default|SetInitialSaplingFruitVariant "NewEnumerator1"))
                          (:NewEnumerator1
                            _)
                          (:NewEnumerator9
                            _)
                          (:NewEnumerator2
                            _)
                          (:NewEnumerator3
                            _)
                          (:NewEnumerator4
                            _)
                          (:NewEnumerator5
                            _)
                          (:NewEnumerator6
                            _)
                          (:NewEnumerator7
                            (if (not (Variables|Default|GetHasSpawnedMythic)) :!label _label_2
                              (CallFunction|CreateServerFruitandBindtoDestroyed _array_index_2 true)
                              (Variables|Default|SetHasSpawnedMythic true)
                              (goto _label Break)))
                          (:NewEnumerator8
                            (goto _label_2)))
                        (CallFunction|CreateServerFruitandBindtoDestroyed _array_index_2))))))
              (:then_4
                (CallFunction|CheckForTooManyFruitsAtRarityLevel true)))
            (elif (>= (Variables|Persistent|GetNumTimesWeeded) (ScalableFloat|GetValueasInteger (Variables|Default|GetMaxWeedingTimes)))
              ;;;[1 1 1] Create new fruits for any missing slots.
              ;;;;[1 1 1] Spawn Mythic Fruit
              (if (not (Variables|Default|GetHasSpawnedMythic))
                (Variables|Default|SetInitialSaplingFruitVariant "NewEnumerator7")
                (if (> _returnvalue 0)
                  (bind _index (Utilities|FlowControl|ForLoopwithBreak 0 (- _returnvalue 1) :!label _label_3)
                    (:LoopBody
                      (bind (_output_slot_4_56babd8945fe8b0c64a28cbc0719d0a6_2 _output_fruit_5_dcca3b6b43057342b19a728973abbf7e_2) (Utilities|Struct|BreakServerSpawnedFruitSlot (Utilities|Array|Get(aref) (Variables|Default|GetServerSpawnedFruitSlot) _index)))
                      (Utilities|IsValid _output_fruit_5_dcca3b6b43057342b19a728973abbf7e_2
                        (:Is%20Not%20Valid
                          (CallFunction|CreateServerFruitandBindtoDestroyed _index true)
                          (Variables|Default|SetHasSpawnedMythic true)
                          (goto _label_3 Break)))))))
              (elif (> _retval 0)
                (if (Variables|Persistent|GetNeedsWeeding)
                  (Development|PrintString "Needs Weeding.  Not creating new Fruit." (Variables|Debug|GetDebug) (Variables|Debug|GetDebug))
                  (else
                    (for _index_2 (range _retval)
                      (Variables|Default|SetInitialSaplingFruitVariant "NewEnumerator1")
                      (bind (_output_slot_4_56babd8945fe8b0c64a28cbc0719d0a6_3 _output_fruit_5_dcca3b6b43057342b19a728973abbf7e_3) (Utilities|Struct|BreakServerSpawnedFruitSlot (Utilities|Array|Get(aref) (Variables|Default|GetServerSpawnedFruitSlot) _index_2)))
                      (Utilities|IsValid _output_fruit_5_dcca3b6b43057342b19a728973abbf7e_3
                        (:Is%20Valid
                          (Development|PrintString (Utilities|String|Append "Fruit already exists in slot: " (Utilities|String|ToString(Integer) _index_2)) (Variables|Debug|GetDebug) (Variables|Debug|GetDebug) "(R=1.000000,G=0.475385,B=0.000000,A=1.000000)" 2e+02))
                        (:Is%20Not%20Valid
                          (CallFunction|CreateServerFruitandBindtoDestroyed _index_2))))))))))
        (:then_1
          (CallFunction|SetupPersistentSaplingFruitStates))))))

(event Custom|SetNotificationTimer
  ;[1 1 1] Notifications
  (Utilities|Time|SetTimerbyEvent (delegate Custom|NotificationTimerExpired) (CallFunction|GetNotificationTime)))

(event Custom|CustomEvent
  (params (DamagedActor :type Actor%20Object%20Reference) (Damage :type Float%20%28single-precision%29) (InstigatedBy :type Controller%20Object%20Reference) (DamageCauser :type Actor%20Object%20Reference) (HitLocation :type Vector) (FHitComponent :type Primitive%20Component%20Object%20Reference) (BoneName :type Name) (Momentum :type Vector))
  (bind _persistentsaplingdata (Variables|RealitySapling|GetPersistentSaplingData))
  (bind _owningplayercontroller (Utilities|Struct|BreakRealitySaplingPersistentData _persistentsaplingdata))
  ;[1 1 1] Initialize Sapling
  ;;[1 1 1] Damaged event to track enemy behavior
  (bind _asfort_player_controller_athena (Utilities|Casting|CastToFortPlayerControllerAthena InstigatedBy))
  (Utilities|IsValid _owningplayercontroller
    (:Is%20Valid
      (if (not (Team|OnSameTeam _asfort_player_controller_athena _owningplayercontroller))
        (CallFunction|OnEnemyDamaged)))))

(event Custom|NotificationTimerExpired
  ;[1 1 1] Notifications
  (Variables|Default|SetEnemyAttacking false)
  (Variables|Default|SetEnemyLooting false)
  (CallFunction|UpdateControllerCompSaplingDataStruct))

(event Custom|OnEnemyDamaged
  ;[1 1 1] Notifications
  (Networking|SwitchHasAuthority
    (:Authority
      (if (CallFunction|NotificationsEnabled)
        (if (Variables|Default|GetOkToNotify)
          (if (not (Variables|Default|GetEnemyAttacking))
            (Variables|Default|SetEnemyAttacking true)
            (CallFunction|UpdateControllerCompSaplingDataStruct)
            (CallFunction|SetNotificationCooldown)
            (CallFunction|SetNotificationTimer)
            (CallFunction|ClientonEnemyDamaged)))))))

(event Custom|Client_OnEnemyDamaged
  ;[1 1 1] Notifications
  (Networking|SwitchHasAuthority
    (:Remote
      (Default|MarkSapling true)
      (Default|CallSaplingDamagedbyEnemyEvent))))

(event Custom|CreateServerFruitAndBindToDestroyed
  (params (Index :type Integer) (IsMythic :type Boolean))
  ;[1 1 1] Initialize Sapling
  ;;[1 1 1] Spawn Fruit Actors
  ;;;[1 1 1] Create Server Fruit And Bind To Destroyed
  (bind _tieredfruitathena (CallFunction|SpawnFruit Index IsMythic))
  (Audio|SpawnSoundatLocation (Variables|Default|GetFruitSpawnSound) (Transformation|GetWorldLocation (Variables|Transformation|GetRootComponent _tieredfruitathena)))
  (Game|BindEventtoOnDestroyed _tieredfruitathena (delegate Custom|On%20Fruit%20Destroyed))
  (Default|BindEventtoVariantorRarityChangedEvent _tieredfruitathena (delegate Custom|OnFruitVariantOrRarityChangedEvent))
  (Networking|ForceNetUpdate))

(event Custom|OnFruitVariantOrRarityChangedEvent
  ;[1 1 1] Initialize Sapling
  ;;[1 1 1] Spawn Fruit Actors
  ;;;[1 1 1] Create Server Fruit And Bind To Destroyed
  (Utilities|IsValid (Variables|Default|GetOwning%20Player%20Controller%20Repped)
    (:Is%20Valid
      (Default|CallSaplingUpdateEvent self))))

(event Custom|On%20Fruit%20Destroyed
  (params (DestroyedActor :type Actor%20Object%20Reference))
  ;[1 1 1] Initialize Sapling
  ;;[1 1 1] Spawn Fruit Actors
  ;;;[1 1 1] Create Server Fruit And Bind To Destroyed
  (if (not (Math|Boolean|ORBoolean (Variables|Default|GetDestroyedForTransplant) (Variables|Default|GetDestroyedFromAging) (CallFunction|IsDead)))
    (bind _astiered_reality_fruit_athena (Utilities|Casting|CastToTiered_RealityFruit_Athena DestroyedActor))
    (for (_array_element _array_index) (Variables|Default|GetServerSpawnedFruitSlot)
      (bind (_array_element_slot_4_56babd8945fe8b0c64a28cbc0719d0a6 _array_element_fruit_5_dcca3b6b43057342b19a728973abbf7e) (Utilities|Struct|BreakServerSpawnedFruitSlot _array_element))
      (if (== _array_element_fruit_5_dcca3b6b43057342b19a728973abbf7e _astiered_reality_fruit_athena)
        (Utilities|Array|SetArrayElem (Variables|Default|GetServerSpawnedFruitSlot) _array_index)
        (bind _astiered_reality_fruit_athena_mythic (Utilities|Casting|CastToTiered_RealityFruit_Athena_Mythic _astiered_reality_fruit_athena)
          (:then
            ;;;;[1 1 1] Destroy because too old (final mythic picked)
            (if (Utilities|Enum|NotEqual(Enum) (Variables|/RealitySeedGameplay/items/RealitySaplingFruits/Tiered_RealityFruit_Athena.Tiered_RealityFruit_Athena_C|GetVariantOrRarity _astiered_reality_fruit_athena_mythic) "NewEnumerator1")
              (CallFunction|InstigateSaplingWither)))
          (:CastFailed
            (CallFunction|StartAddFruitTimerIfNeeded)
            (Networking|ForceNetUpdate)))))
    (CallFunction|SetupPersistentSaplingFruitStates)))

(event Custom|On%20End%20Of%20Life)

(event Custom|OnEnemyHarvested
  ;[1 1 1] Notifications
  (Networking|SwitchHasAuthority
    (:Authority
      (if (CallFunction|NotificationsEnabled)
        (if (Variables|Default|GetOkToNotify)
          (if (not (Variables|Default|GetEnemyLooting))
            (Variables|Default|SetEnemyLooting true)
            (CallFunction|UpdateControllerCompSaplingDataStruct)
            (CallFunction|SetNotificationCooldown)
            (CallFunction|SetNotificationTimer)
            (CallFunction|ClientonEnemyHarvested)))))))

(event Custom|CreateFruits
  ;[1 1 1] Initialize Sapling
  ;;[1 1 1] Spawn Fruit Actors
  ;;;[1 1 1] Create Server Fruit And Bind To Destroyed
  (CallFunction|SpawnMissingFruitActors)
  (Utilities|Time|ClearandInvalidateTimerbyHandle (Variables|Default|GetCreateFruitTimerHandle)))

(event Custom|Client_OnEnemyHarvested
  ;[1 1 1] Notifications
  (Networking|SwitchHasAuthority
    (:Remote
      (Default|MarkSapling true)
      (Default|CallSaplingHarvestedbyEnemyEvent))))

(event Custom|SpawnWeedsImmediate
  ;[1 1 1] Initialize Sapling
  ;;[1 1 1] Spawn Weeds if Needed
  ;;;[1 1 1] spawn weed immediately
  (if (Math|Boolean|ANDBoolean (<= (Variables|Persistent|GetNumTimesWeeded) (ScalableFloat|GetValueasInteger (Variables|Default|GetMaxWeedingTimes))) (== (Utilities|Array|Length (Variables|Default|GetServerWeeds)) 0) (not (CallFunction|IsDead)))
    (bind _returnvalue (Game|SpawnActorfromClass :Class "/RealitySeedGameplay/Environment/Foliage/B_RealitySaplingsWeeds_Spawner.B_RealitySaplingsWeeds_Spawner_C" :SpawnTransform (Transformation|GetActorTransform) :CollisionHandlingOverride "AlwaysSpawn" :TransformScaleMethod "MultiplyWithRoot" :Owner (Actor|GetOwner) :Sapling self :UseWeightedSpawnClassList false :Instigator (Variables|Actor|GetInstigator)))
    (Variables|Default|SetWeedSpawner _returnvalue)
    (Game|BindEventtoOnDestroyed _returnvalue (delegate Custom|WeedSpawnerDestroyed))))

(event Custom|Instigate%20Sapling%20Wither
  ;[1 1 1] Initialize Sapling
  ;;[1 1 1] Instigate Sapling Wither
  (Variables|Default|SetSaplingNurturedCountThisMatch (+ (ScalableFloat|GetValueasInteger (Variables|Default|GetMaxWeedingTimes)) 1))
  (Variables|Default|SetDestroyedFromAging true)
  (Development|PrintString "Sapling is too old (mythic picked).  Withering...")
  (CallFunction|SetShouldEnterGhostStateonDeath false) ; Allow to fully die
  (Networking|ForceNetUpdate)
  (Actor|SetLifeSpan 3.0)
  (CallFunction|OnEndOfLife)
  (for _array_element (Variables|Default|GetServerSpawnedFruitSlot)
    (bind (_array_element_slot_4_56babd8945fe8b0c64a28cbc0719d0a6 _array_element_fruit_5_dcca3b6b43057342b19a728973abbf7e) (Utilities|Struct|BreakServerSpawnedFruitSlot _array_element))
    (Utilities|IsValid _array_element_fruit_5_dcca3b6b43057342b19a728973abbf7e
      (:Is%20Valid
        (Class|TieredRealityFruitAthena|ScaleOutAndDie _array_element_fruit_5_dcca3b6b43057342b19a728973abbf7e)
        (Networking|ForceNetUpdate _array_element_fruit_5_dcca3b6b43057342b19a728973abbf7e)))))

(event Custom|SpawnWeedsIfNeeded
  ;[1 1 1] Initialize Sapling
  ;;[1 1 1] Spawn Weeds if Needed
  (Networking|SwitchHasAuthority
    (:Authority
      (if (Variables|Persistent|GetNeedsWeeding)
        (if (<= (Utilities|Array|Length (Variables|Default|GetServerWeeds)) 0)
          (CallFunction|SpawnWeedsImmediate))))))

(event Custom|WeedSpawnerDestroyed
  (params (DestroyedActor :type Actor%20Object%20Reference))
  ;[1 1 1] Initialize Sapling
  ;;[1 1 1] Spawn Weeds if Needed
  ;;;[1 1 1] spawn weed immediately
  (if (or (Variables|Default|GetDestroyedForTransplant) (Variables|Default|GetDestroyedFromAging))
    _
    (elif (ScalableFloat|GetValueasBool (Variables|Default|GetAutoWeedWhenWeedsCantBePlaced))
      (if (== (Utilities|Array|Length (Variables|Default|GetServerWeeds)) 0)
        (Fortnite|FortShippingLog 0 "Could not spawn weeds : triggering weeding automatically so sapling doesn't stagnate.  This should happen rarely if ever.")
        (Variables|Default|SetSaplingNurturedCountThisMatch (+ (Variables|Default|GetSaplingNurturedCountThisMatch) 1))
        (Networking|ForceNetUpdate)))))

(event Custom|SpawnMushrooms
  ;[1 1 1] Initialize Sapling
  ;;[1 1 1] Spawn Consumables if needed
  (Networking|SwitchHasAuthority
    (:Authority
      (if (>= (Variables|Persistent|GetNumTimesWeeded) (ScalableFloat|GetValueasInteger (Variables|Default|GetMaturityLevelWhenGroundItemsShowUp)))
        (Utilities|FlowControl|DoOnce false
          (:Completed
            (Game|SpawnActorfromClass :Class "/RealitySeedGameplay/items/RealitySaplings/EQS/B_RealitySaplingsMushrooms_Spawner.B_RealitySaplingsMushrooms_Spawner_C" :SpawnTransform (Transformation|GetActorTransform) :CollisionHandlingOverride "Undefined" :TransformScaleMethod "MultiplyWithRoot" :Sapling self :ClassToSpawn "/RealitySeedGameplay/items/RealitySaplings/EQS/CBGA_SilkyMoose_ScaleIn.CBGA_SilkyMoose_ScaleIn_C" :UseWeightedSpawnClassList false :Instigator (Variables|Actor|GetInstigator))))))))

(event Custom|OnSaplingNurtured
  ;[1 1 1] Sapling has been nurtured
  ;;[1 1 1] Audio and FX Here: Make tree celebrate that it has been nurtured (i.e. happy\glow for a few seconds.
  (Development|PrintString "Sapling Nurtured")
  (Utilities|Time|SetTimerbyFunctionName 0 "Notify Fruits Upgraded" (ScalableFloat|GetValueatLevel (Variables|Default|GetDelayForNurtureFruitUpgrade)))
  (Networking|SwitchHasAuthority
    (:Remote
      (if (or (Variables|Default|GetDestroyedForTransplant) (Variables|Default|GetDestroyedFromAging))
        (CallFunction|EventFinishedSound (Variables|Default|GetWitheringSound))
        (else
          (CallFunction|EventFinishedSound (Variables|Default|GetNurturedSound)))))))

(event Custom|AgeUpSapling
  (bind (_persistentsaplingdata_owningplayercontroller _persistentsaplingdata_location _persistentsaplingdata_rotation _persistentsaplingdata_nurturinglevel _persistentsaplingdata_maturitylevel _persistentsaplingdata_plantingtime _persistentsaplingdata_lastweedingtime _persistentsaplingdata_fruitstates) (Utilities|Struct|BreakRealitySaplingPersistentData (Variables|GetPersistentSaplingData_OwningPlayerController)))
  (bind (_persistentsaplingdata_owningplayercontroller_2 _persistentsaplingdata_location_2 _persistentsaplingdata_rotation_2 _persistentsaplingdata_nurturinglevel_2 _persistentsaplingdata_maturitylevel_2) (Utilities|Struct|BreakRealitySaplingPersistentData (Variables|GetPersistentSaplingData_OwningPlayerController)))
  ;[0.7716 1 0.7562] Age Up Sapling
  (Networking|SwitchHasAuthority
    (:Authority
      (if (not (Variables|Default|GetOverrride%20Sapling%20Persistence))
        (Utilities|FlowControl|DoOnce false
          (:Completed
            (Variables|SetPersistentSaplingData_OwningPlayerController (Utilities|Struct|MakeRealitySaplingPersistentData _persistentsaplingdata_owningplayercontroller _persistentsaplingdata_location _persistentsaplingdata_rotation _persistentsaplingdata_nurturinglevel (+ _persistentsaplingdata_maturitylevel 1) _persistentsaplingdata_plantingtime _persistentsaplingdata_lastweedingtime _persistentsaplingdata_fruitstates))
            (Development|PrintString (Utilities|String|Append "Aged up to " (Utilities|String|ToString(Integer) _persistentsaplingdata_maturitylevel_2)) (Variables|Debug|GetDebug) (Variables|Debug|GetDebug))))))))

(event Custom|EventFinishedSound
  (params (SoundToPlay :type Sound%20Base%20Object%20Reference))
  ;[1 1 1] Sapling has been nurtured
  (Audio|SpawnSoundatLocation SoundToPlay (Transformation|GetWorldLocation (Variables|Debug|GetStaticMeshComponent))))

(event Custom|On%20Sapling%20Partially%20Nurtured
  ;[1 1 1] Sapling has been nurtured
  (Default|CallSaplingUpdateEvent self))

(event Custom|EventGrowthSound
  ;[1 1 1] Audio - Fruit Growth
  (Utilities|FlowControl|DoOnce false :!label _label
    (:Completed
      (Audio|SpawnSoundatLocation (Variables|Default|GetFruitGrowSound) (Transformation|GetWorldLocation (Variables|Debug|GetStaticMeshComponent)))
      (Utilities|FlowControl|Delay 0.5)
      (goto _label Reset))))

(event EventOnSaplingRegistered
  (CallFunction|UpdateControllerCompSaplingDataStruct))

(event Game|Damage|EventPointDamage
  (params (Damage :type Float%20%28single-precision%29) (DamageType :type Damage%20Type%20Object%20Reference) (HitLocation :type Vector :default "0, 0, 0") (HitNormal :type Vector :default "0, 0, 0") (HitComponent :type Primitive%20Component%20Object%20Reference) (BoneName :type Name :default "None") (ShotFromDirection :type Vector :default "0, 0, 0") (InstigatedBy :type Controller%20Object%20Reference) (DamageCauser :type Actor%20Object%20Reference) (HitInfo :type Hit%20Result%20Structure%20%28by%20ref%29))
  (Development|PrintString (Utilities|String|Append "POINT DAMAGE: " (Utilities|String|ToString(Float) Damage) "   :    " (Utilities|GetDisplayName DamageType)) (Variables|Debug|GetDebug) (Variables|Debug|GetDebug)))

(event Game|Damage|EventRadialDamage
  (params (DamageReceived :type Float%20%28single-precision%29) (DamageType :type Damage%20Type%20Object%20Reference) (Origin :type Vector :default "0, 0, 0") (HitInfo :type Hit%20Result%20Structure%20%28by%20ref%29) (InstigatedBy :type Controller%20Object%20Reference) (DamageCauser :type Actor%20Object%20Reference))
  (Development|PrintString "RADIAL DAMAGE")
  (Development|PrintString (Utilities|String|Append "RADIAL DAMAGE: " (Utilities|String|ToString(Float) DamageReceived) "   :    " (Utilities|GetDisplayName DamageType)) (Variables|Debug|GetDebug) (Variables|Debug|GetDebug)))

(event Game|Damage|EventAnyDamage
  (params (Damage :type Float%20%28single-precision%29) (DamageType :type Damage%20Type%20Object%20Reference) (InstigatedBy :type Controller%20Object%20Reference) (DamageCauser :type Actor%20Object%20Reference))
  (Development|PrintString "ANY DAMAGE")
  (Development|PrintString (Utilities|String|Append "RADIAL DAMAGE: " (Utilities|String|ToString(Float) Damage) "   :    " (Utilities|GetDisplayName DamageType)) (Variables|Debug|GetDebug) (Variables|Debug|GetDebug))
  (Audio|SpawnSoundatLocation (Variables|Default|GetDamagedSound) (Transformation|GetWorldLocation (Variables|Debug|GetStaticMeshComponent))))

(event EventDestroyed
  (Development|PrintString "Destroyed" (Variables|Debug|GetDebug) (Variables|Debug|GetDebug))
  (if (Variables|Default|GetDestroyedFromAging)
    (|CallParentFunction)))

(event Curie|EventOnCurieStateAttached
  (params (CurieContainerHandle :type Curie%20Container%20Handle%20Structure) (StateTag :type Gameplay%20Tag%20Structure))
  ;[1 1 1] Watered (Handle Starting Wet State (curie Attached))
  (Networking|SwitchHasAuthority
    (:Authority
      (Development|PrintString "CurieStateAttached" (Variables|Debug|GetDebug) (Variables|Debug|GetDebug))
      (if (GameplayTags|MatchesTag StateTag "(TagName=\"Curie.State.ElementalStatus.Wet\")" true)
        (if (< (Variables|Default|GetNumTimesWateredThisMatch) (ScalableFloat|GetValueatLevel (Variables|Default|GetMaxTimesASaplingCanBeWateredPerMatch)))
          (Variables|Default|SetIsBeingWatered true)
          (Variables|Default|SetLastWateredTime (Utilities|Time|GetTimeSeconds))
          (if (< (Variables|Default|GetSecondsWateredThisMatch) (ScalableFloat|GetValueatLevel (Variables|Default|GetNumSecondsWateredForReaction)))
            (if (>= (Variables|Default|GetSecondsWateredThisMatch) (ScalableFloat|GetValueatLevel (Variables|Default|GetNumSecondsWateredForReaction)))
              (CallFunction|WaterSapling)
              (else
                _)))
          (CallFunction|StartWateredTimer))))))

(event Custom|Start%20Watered%20Timer
  ;[1 1 1] Watered (Handle Starting Wet State (curie Attached))
  (bind _returnvalue (Utilities|Time|SetTimerbyEvent (delegate Custom|WaterSapling) (- (ScalableFloat|GetValueatLevel (Variables|Default|GetNumSecondsWateredForReaction)) (Variables|Default|GetSecondsWateredThisMatch))))
  (Variables|Default|SetWaterTimerHandle _returnvalue))

(event Custom|WaterSapling
  ;[1 1 1] Watered (Handle Starting Wet State (curie Attached))
  (if (not (CallFunction|IsDead))
    (Math|Integer|IncrementInt (Variables|Default|GetNumTimesWateredThisMatch)
      (:%20%20
        (CallFunction|OnSaplingPartiallyNurtured)
        (Variables|Default|SetSecondsWateredThisMatch 0.0)
        (for _array_element (Variables|Default|GetServerSpawnedFruitSlot)
          (bind (_array_element_slot_4_56babd8945fe8b0c64a28cbc0719d0a6 _array_element_fruit_5_dcca3b6b43057342b19a728973abbf7e) (Utilities|Struct|BreakServerSpawnedFruitSlot _array_element))
          (Utilities|IsValid _array_element_fruit_5_dcca3b6b43057342b19a728973abbf7e
            (:Is%20Valid
              (if (not (Variables|/RealitySeedGameplay/items/RealitySaplingFruits/Tiered_RealityFruit_Athena.Tiered_RealityFruit_Athena_C|GetIsMythicFruit _array_element_fruit_5_dcca3b6b43057342b19a728973abbf7e))
                (Class|TieredRealityFruitAthena|SettoUseShortTimes _array_element_fruit_5_dcca3b6b43057342b19a728973abbf7e)))))
        (if (< (Variables|Default|GetNumTimesWateredThisMatch) (ScalableFloat|GetValueasInteger (Variables|Default|GetMaxTimesASaplingCanBeWateredPerMatch)))
          (if (Variables|Default|GetIsBeingWatered)
            (CallFunction|StartWateredTimer)))))))

(event Curie|EventOnCurieStateDetached
  (params (CurieContainerHandle :type Curie%20Container%20Handle%20Structure) (StateTag :type Gameplay%20Tag%20Structure))
  ;[1 1 1] End Watered (Handle Ending Wet State\Dry (curie Dettached))
  (Networking|SwitchHasAuthority
    (:Authority
      (Development|PrintString "CurieStateDetached" (Variables|Debug|GetDebug) (Variables|Debug|GetDebug))
      (if (GameplayTags|MatchesTag StateTag "(TagName=\"Curie.State.ElementalStatus.Wet\")" true)
        (Variables|Default|SetIsBeingWatered false)
        (if (< (Variables|Default|GetNumTimesWateredThisMatch) (ScalableFloat|GetValueasInteger (Variables|Default|GetMaxTimesASaplingCanBeWateredPerMatch)))
          (Variables|Default|SetSecondsWateredThisMatch (+ (Variables|Default|GetSecondsWateredThisMatch) (- (Utilities|Time|GetTimeSeconds) (Variables|Default|GetLastWateredTime))))
          (if (>= (Variables|Default|GetSecondsWateredThisMatch) (ScalableFloat|GetValueatLevel (Variables|Default|GetNumSecondsWateredForReaction)))
            (CallFunction|WaterSapling))
          (Utilities|Time|ClearandInvalidateTimerbyHandle (Variables|Default|GetWaterTimerHandle)))))))

(event EventEnterGhostState
  ;[0.4597 1 0.8277] c++ ghost state hook
  (CallFunction|OnSaplingKilled))

(event Custom|OnSaplingKilled
  (bind (_saplingbpreppeddata_saplingage_4_5368fd2741dc1813a345149d36ca3e98 _saplingbpreppeddata_revivecount_5_ec5af65440fb691c7b8724bd7cdfa3e1) (Utilities|Struct|BreakSaplingBPReppedData (Variables|GetSaplingBPReppedData_SaplingAge_4_5368FD2741DC1813A345149D36CA3E98)))
  (bind _saplingbpreppeddata_saplingage_4_5368fd2741dc1813a345149d36ca3e98_2 (Utilities|Struct|BreakSaplingBPReppedData (Variables|GetSaplingBPReppedData_SaplingAge_4_5368FD2741DC1813A345149D36CA3E98)))
  ;[1 1 1] Sapling Death Event called when saplings die\are killed
  (if (not (Math|Boolean|ORBoolean (Variables|Default|GetHasDied) (Variables|Default|GetDestroyedForTransplant) (Variables|Default|GetDestroyedFromAging)))
    (Variables|Default|SetHasDied true)
    (Variables|SetSaplingBPReppedData_SaplingAge_4_5368FD2741DC1813A345149D36CA3E98 (Utilities|Struct|MakeSaplingBPReppedData _saplingbpreppeddata_saplingage_4_5368fd2741dc1813a345149d36ca3e98_2 (+ _saplingbpreppeddata_revivecount_5_ec5af65440fb691c7b8724bd7cdfa3e1 1)))
    (Networking|SwitchHasAuthority
      (:Authority
        (CallFunction|UpdateControllerCompSaplingDataStruct) ; Update struct on controller component for UI purposes
      ))))

(event Custom|CheckForSaplingNewlyTransplanted
  (Networking|SwitchHasAuthority
    (:Remote
      (if (Variables|Default|GetSaplingTransplanted)
        (CallFunction|OnSaplingNewlyTransplanted)))))

(event Custom|On%20Sapling%20Transplanting%20Away
  (Development|PrintString "On Sapling Transplanting Away")
  (Audio|Components|Audio|FadeOut (Variables|Default|GetAmbientLoop) 0.3 0.0 "Linear" :!class /Script/Engine.AudioComponent))

(event Custom|On%20Weeds%20Killed)

(event EventForceSpawnWeeds
  (CallFunction|SetNeedsWeeding)
  (CallFunction|SpawnWeedsifNeeded))

(event Custom|CheckMatchStartHotFixFruitStateOverrides
  (bind _overrideminadultfruitsatstartofmatch (Variables|Default|GetOverrideMinAdultFruitsAtStartOfMatch))
  (bind _returnvalue (ScalableFloat|GetValueasInteger _overrideminadultfruitsatstartofmatch))
  (bind _overrideminbabyfruitsatstartofmatch (Variables|Default|GetOverrideMinBabyFruitsAtStartOfMatch))
  (bind _returnvalue_2 (ScalableFloat|GetValueasInteger _overrideminbabyfruitsatstartofmatch))
  (bind _overrideminyouthfruitsatstartofmatch (Variables|Default|GetOverrideMinYouthFruitsAtStartOfMatch))
  (bind _returnvalue_3 (ScalableFloat|GetValueasInteger _overrideminyouthfruitsatstartofmatch))
  (bind _overrideminadultfruitsatstartofmatch_2 (Variables|Default|GetOverrideMinAdultFruitsAtStartOfMatch))
  (bind _returnvalue_4 (ScalableFloat|GetValueasInteger _overrideminadultfruitsatstartofmatch_2))
  (bind _overrideminbabyfruitsatstartofmatch_2 (Variables|Default|GetOverrideMinBabyFruitsAtStartOfMatch))
  (bind _returnvalue_5 (ScalableFloat|GetValueasInteger _overrideminbabyfruitsatstartofmatch_2))
  (bind _overrideminyouthfruitsatstartofmatch_2 (Variables|Default|GetOverrideMinYouthFruitsAtStartOfMatch))
  (bind _returnvalue_6 (ScalableFloat|GetValueasInteger _overrideminyouthfruitsatstartofmatch_2))
  (bind _returnvalue_7 (> _returnvalue_2 0))
  (bind _returnvalue_8 (> _returnvalue_3 0))
  (bind _returnvalue_9 (> _returnvalue_4 0))
  (bind _fruitstatecount (Variables|Default|GetfruitStateCount))
  (bind _fruitstatecount_2 (Variables|Default|GetfruitStateCount))
  (bind _fruitstatecount_3 (Variables|Default|GetfruitStateCount))
  ;[1 1 1] Check Override HotFixes
  (if (Variables|Persistent|GetNeedsWeeding)
    _
    (elif (Math|Boolean|ORBoolean _returnvalue_7 _returnvalue_8 _returnvalue_9)
      (Utilities|FlowControl|Sequence
        (:then_0
          (if _returnvalue_7
            (Variables|Default|SetfruitStateCount 0)
            (for _array_element (Variables|Default|GetPersistent%20Sapling%20Data%20Fruit%20States)
              (switch Utilities|FlowControl|Switch|SwitchonSaplingFruitVariants (Utilities|Enum|BytetoEnumSaplingFruitVariants (Math|Conversions|ToByte(Integer) _array_element))
                (:NewEnumerator0
                  _)
                (:NewEnumerator1
                  _))
              (Math|Integer|IncrementInt (Variables|Default|GetfruitStateCount)))
            (if (> _returnvalue_5 _fruitstatecount_2)
              (for _ (range (- _returnvalue_5 _fruitstatecount_2))
                (bind (_array_element_2 _array_index) (Utilities|Array|ForEachLoopwithBreak (Variables|Default|GetPersistent%20Sapling%20Data%20Fruit%20States) :!label _label)
                  (:LoopBody
                    (switch Utilities|FlowControl|Switch|SwitchonSaplingFruitVariants (Utilities|Enum|BytetoEnumSaplingFruitVariants (Math|Conversions|ToByte(Integer) _array_element_2))
                      (:NewEnumerator9
                        _)
                      (:NewEnumerator2
                        _)
                      (:NewEnumerator3
                        _)
                      (:NewEnumerator4
                        _)
                      (:NewEnumerator5
                        _)
                      (:NewEnumerator6
                        _))
                    (Utilities|Array|SetArrayElem (Variables|Default|GetPersistent%20Sapling%20Data%20Fruit%20States) _array_index 1)
                    (goto _label Break)))))))
        (:then_1
          (if _returnvalue_8
            (Variables|Default|SetfruitStateCount 0)
            (for _array_element_3 (Variables|Default|GetPersistent%20Sapling%20Data%20Fruit%20States)
              (switch Utilities|FlowControl|Switch|SwitchonSaplingFruitVariants (Utilities|Enum|BytetoEnumSaplingFruitVariants (Math|Conversions|ToByte(Integer) _array_element_3))
                (:NewEnumerator9
                  (Math|Integer|IncrementInt (Variables|Default|GetfruitStateCount)))))
            (if (> _returnvalue_6 _fruitstatecount_3)
              (for _ (range (- _returnvalue_6 _fruitstatecount_3))
                (bind (_array_element_4 _array_index_2) (Utilities|Array|ForEachLoopwithBreak (Variables|Default|GetPersistent%20Sapling%20Data%20Fruit%20States) :!label _label_2)
                  (:LoopBody
                    (switch Utilities|FlowControl|Switch|SwitchonSaplingFruitVariants (Utilities|Enum|BytetoEnumSaplingFruitVariants (Math|Conversions|ToByte(Integer) _array_element_4))
                      (:NewEnumerator0
                        _)
                      (:NewEnumerator1
                        _)
                      (:NewEnumerator2
                        _)
                      (:NewEnumerator3
                        _)
                      (:NewEnumerator4
                        _)
                      (:NewEnumerator5
                        _)
                      (:NewEnumerator6
                        _))
                    (Utilities|Array|SetArrayElem (Variables|Default|GetPersistent%20Sapling%20Data%20Fruit%20States) _array_index_2 2)
                    (goto _label_2 Break)))))))
        (:then_2
          (if _returnvalue_9
            (Variables|Default|SetfruitStateCount 0)
            (for _array_element_5 (Variables|Default|GetPersistent%20Sapling%20Data%20Fruit%20States)
              (switch Utilities|FlowControl|Switch|SwitchonSaplingFruitVariants (Utilities|Enum|BytetoEnumSaplingFruitVariants (Math|Conversions|ToByte(Integer) _array_element_5))
                (:NewEnumerator2
                  _)
                (:NewEnumerator3
                  _)
                (:NewEnumerator4
                  _)
                (:NewEnumerator5
                  _)
                (:NewEnumerator6
                  _))
              (Math|Integer|IncrementInt (Variables|Default|GetfruitStateCount)))
            (if (> _returnvalue _fruitstatecount)
              (for _ (range (- _returnvalue _fruitstatecount))
                (bind (_array_element_6 _array_index_3) (Utilities|Array|ForEachLoopwithBreak (Variables|Default|GetPersistent%20Sapling%20Data%20Fruit%20States) :!label _label_3)
                  (:LoopBody
                    (switch Utilities|FlowControl|Switch|SwitchonSaplingFruitVariants (Utilities|Enum|BytetoEnumSaplingFruitVariants (Math|Conversions|ToByte(Integer) _array_element_6))
                      (:NewEnumerator0
                        _)
                      (:NewEnumerator1
                        _)
                      (:NewEnumerator9
                        _))
                    (Variables|Default|SetGrayAsInt (Math|Conversions|ToInteger(Byte) (Variables|Default|GetGray)))
                    (Variables|Default|SetMythicAsBabyInt (Math|Conversions|ToInteger(Byte) (Variables|Default|GetMythicBaby)))
                    (Variables|Default|SetGoldAsInt (Math|Conversions|ToInteger(Byte) (Variables|Default|GetGold)))
                    (Variables|Default|SetTargetAsInt (+ (Variables|Persistent|GetNumTimesWeeded) (Variables|Default|GetGrayAsInt)))
                    (if (>= (Variables|Default|GetTargetAsInt) (Variables|Default|GetMythicAsBabyInt))
                      (Variables|Default|SetTargetAsInt (Variables|Default|GetGoldAsInt)))
                    (Utilities|Array|SetArrayElem (Variables|Default|GetPersistent%20Sapling%20Data%20Fruit%20States) _array_index_3 (Variables|Default|GetTargetAsInt))
                    (goto _label_3 Break)))))))))))

(fn AddServerConsumable
  (params (Consumable :type Actor%20Object%20Reference))
  (Utilities|IsValid Consumable
    (:Is%20Valid
      (Utilities|Array|AddUnique (Variables|Default|GetServerConsumables) Consumable))))

(fn AddServerWeed
  (params (Weed :type Actor%20Object%20Reference))
  (bind _serverweeds (Variables|Default|GetServerWeeds))
  (Utilities|IsValid Weed
    (:Is%20Valid
      (Utilities|Array|AddUnique _serverweeds Weed)
      (Variables|Default|SetNumCurrentWeeds (Utilities|Array|Length _serverweeds)))))

(fn CalcNeedsWeeding
  (bind (_newparam_year _newparam_month _newparam_day) (Math|DateTime|BreakDateTime (CallFunction|GetUTCTimewithOffsetAdded)))
  (bind _saplingbpreppeddata_saplingage_4_5368fd2741dc1813a345149d36ca3e98 (Utilities|Struct|BreakSaplingBPReppedData (Variables|GetSaplingBPReppedData_SaplingAge_4_5368FD2741DC1813A345149D36CA3E98)))
  (bind (_lastweededdatetime_year _lastweededdatetime_month _lastweededdatetime_day) (Math|DateTime|BreakDateTime (Variables|GetLastWeededDateTime_Year)))
  (switch Utilities|FlowControl|Switch|SwitchonSaplingAgeVariants _saplingbpreppeddata_saplingage_4_5368fd2741dc1813a345149d36ca3e98
    (:NewEnumerator1
      _)
    (:NewEnumerator2
      _)
    (:NewEnumerator3
      _))
  (if (or (<= (Variables|Persistent|GetNumTimesWeeded) (ScalableFloat|GetValueasInteger (Variables|Default|GetMinNurturingLevel))) false)
    (CallFunction|SetNeedsWeeding)
    (elif (>= (Variables|Persistent|GetNumTimesWeeded) (ScalableFloat|GetValueasInteger (Variables|Default|GetMaxWeedingTimes)))
      _
      (elif (Math|Boolean|ORBoolean (!= _lastweededdatetime_year _newparam_year) (!= _lastweededdatetime_month _newparam_month) (!= _lastweededdatetime_day _newparam_day))
        (CallFunction|SetNeedsWeeding)))))

(fn CheckforTooManyFruitsatRarityLevel
  (params (ForceSilent :type Boolean))
  (Networking|SwitchHasAuthority
    (:Authority
      (if (not (>= (Variables|Persistent|GetNumTimesWeeded) (ScalableFloat|GetValueasInteger (Variables|Default|GetMaxWeedingTimes))))
        (LocalVariables|SetNonMythicFruitCount 0)
        (for _array_element (Variables|Default|GetServerSpawnedFruitSlot)
          (bind (_array_element_slot_4_56babd8945fe8b0c64a28cbc0719d0a6 _array_element_fruit_5_dcca3b6b43057342b19a728973abbf7e) (Utilities|Struct|BreakServerSpawnedFruitSlot _array_element))
          (Utilities|IsValid _array_element_fruit_5_dcca3b6b43057342b19a728973abbf7e
            (:Is%20Valid
              (if (not (Variables|/RealitySeedGameplay/items/RealitySaplingFruits/Tiered_RealityFruit_Athena.Tiered_RealityFruit_Athena_C|GetIsMythicFruit _array_element_fruit_5_dcca3b6b43057342b19a728973abbf7e))
                (Math|Integer|IncrementInt (LocalVariables|GetNonMythicFruitCount))))))
        (if (> (LocalVariables|GetNonMythicFruitCount) (CallFunction|GetNumFruitsToGrowatCurrentRarity))
          (Utilities|FlowControl|Sequence
            (:then_0
              (LocalVariables|SetNumFruitsToRemove (- (LocalVariables|GetNonMythicFruitCount) (CallFunction|GetNumFruitsToGrowatCurrentRarity)))
              (LocalVariables|SetNumFruitsRemoved 0))
            (:then_1
              (bind (_arrayindex _arrayelement) (Utilities|Array|ReverseforEachLoop (Variables|Default|GetServerSpawnedFruitSlot))
                (:LoopBody
                  (if (< (LocalVariables|GetNumFruitsRemoved) (LocalVariables|GetNumFruitsToRemove))
                    (bind (_arrayelement_slot_4_56babd8945fe8b0c64a28cbc0719d0a6 _arrayelement_fruit_5_dcca3b6b43057342b19a728973abbf7e) (Utilities|Struct|BreakServerSpawnedFruitSlot _arrayelement))
                    (Utilities|IsValid _arrayelement_fruit_5_dcca3b6b43057342b19a728973abbf7e
                      (:Is%20Valid
                        (if (Variables|/RealitySeedGameplay/items/RealitySaplingFruits/Tiered_RealityFruit_Athena.Tiered_RealityFruit_Athena_C|GetIsMythicFruit _arrayelement_fruit_5_dcca3b6b43057342b19a728973abbf7e)
                          _
                          (elif (ScalableFloat|GetValueasBool (Variables|Default|GetAutoLootFruitsWhenTooMany))
                            (if (Variables|GetForceSilent)
                              (Class|TieredRealityFruitAthena|ScaleOutAndDie _arrayelement_fruit_5_dcca3b6b43057342b19a728973abbf7e :!label _label)
                              (else
                                (Class|TieredRealityFruitAthena|AutoLootFruit _arrayelement_fruit_5_dcca3b6b43057342b19a728973abbf7e (- (Variables|Persistent|GetNumTimesWeeded) 1))))
                            (Math|Integer|IncrementInt (LocalVariables|GetNumFruitsRemoved))))
                        (goto _label)))))))))))))

(fn GetFruitSpawnTransform
  (params (FruitIndex :type Integer))
  (bind _active_static_mesh_comp (Variables|Default|GetActive%20Static%20Mesh%20Comp))
  (bind (_returnvalue_location _returnvalue_rotation) (Math|Transform|BreakTransform (Transformation|GetSocketTransform _active_static_mesh_comp "FruitSocket3" "RTS_World")))
  (bind (_x _y _z) (Math|Rotator|GetAxes _returnvalue_rotation))
  (bind _active_static_mesh_comp_2 (Variables|Default|GetActive%20Static%20Mesh%20Comp))
  (bind (_returnvalue_location_2 _returnvalue_rotation_2) (Math|Transform|BreakTransform (Transformation|GetSocketTransform _active_static_mesh_comp_2 "FruitSocket1" "RTS_World")))
  (bind (_x_2 _y_2 _z_2) (Math|Rotator|GetAxes _returnvalue_rotation_2))
  (bind (_returnvalue_roll _returnvalue_pitch _returnvalue_yaw) (Math|Rotator|BreakRotator (Transformation|GetActorRotation)))
  (bind _active_static_mesh_comp_3 (Variables|Default|GetActive%20Static%20Mesh%20Comp))
  (bind (_returnvalue_location_3 _returnvalue_rotation_3) (Math|Transform|BreakTransform (Transformation|GetSocketTransform _active_static_mesh_comp_3 "FruitSocket2" "RTS_World")))
  (bind (_x_3 _y_3 _z_3) (Math|Rotator|GetAxes _returnvalue_rotation_3))
  (if (Utilities|Enum|Equal(Enum) (CallFunction|GetSaplingAgeVariant) "NewEnumerator1")
    (Variables|Default|SetActive%20Static%20Mesh%20Comp (Variables|Default|GetYouthStaticMesh))
    (else
      (Variables|Default|SetActive%20Static%20Mesh%20Comp (Variables|Debug|GetStaticMeshComponent))))
  (Utilities|IsValid (Variables|Default|GetActive%20Static%20Mesh%20Comp)
    (:Is%20Valid
      (switch int FruitIndex
        (:Default
          _)
        (:0
          (if (Transformation|DoesSocketExist _active_static_mesh_comp_2 "FruitSocket1")
            (return (+ _returnvalue_location_2 (* _z_2 (Variables|Default|GetMoveOnZAxis))) _returnvalue_rotation_2 "1.000000,1.000000,1.000000")
            (else
              _)))
        (:1
          (if (Transformation|DoesSocketExist _active_static_mesh_comp_3 "FruitSocket2")
            (return (+ _returnvalue_location_3 (* _z_3 (Variables|Default|GetMoveOnZAxis))) _returnvalue_rotation_3 "1.000000,1.000000,1.000000")
            (else
              _)))
        (:2
          (if (Transformation|DoesSocketExist _active_static_mesh_comp "FruitSocket3")
            (return (+ _returnvalue_location (* _z (Variables|Default|GetMoveOnZAxis))) _returnvalue_rotation "1.000000,1.000000,1.000000")
            (else
              _)))))
    (:Is%20Not%20Valid
      _))
  (return (Utilities|Operators|Add (Transformation|GetActorLocation) (Math|Vector|RotateVectorAroundAxis (Math|Vector|MakeVector (Variables|Default|GetFruitHorizDistanceFromCenter)) (+ (* (/ (Math|Conversions|ToFloat(Integer) FruitIndex) (Math|Conversions|ToFloat(Integer) (ScalableFloat|GetValueasInteger (Variables|Default|GetNumFruitsToGrow)))) 3.6e+02) (+ _returnvalue_yaw (Variables|Default|GetFruitAngleOffset))) "0, 0,1.000000") (Math|Vector|MakeVector 0.0 0.0 (Variables|Default|GetFruitStartDistanceFromGround))) 0.0 0.0 (Math|Random|RandomFloatinRange 0.0 3.6e+02) "1.000000,1.000000,1.000000"))

(fn GetNumFruitstoGrowAtCurrentRarity
  (bind (_owningplayercontroller _location _rotation _nurturinglevel) (Utilities|Struct|BreakRealitySaplingPersistentData (Variables|RealitySapling|GetPersistentSaplingData)))
  (bind _numfruitstogrow_legendary (Variables|Default|GetNumFruitsToGrow_Legendary))
  (bind _returnvalue (ScalableFloat|GetValueasInteger _numfruitstogrow_legendary))
  (bind _numfruitstogrow_common (Variables|Default|GetNumFruitsToGrow_Common))
  (bind _returnvalue_2 (ScalableFloat|GetValueasInteger _numfruitstogrow_common))
  (bind _numfruitstogrow_uncommon (Variables|Default|GetNumFruitsToGrow_Uncommon))
  (bind _returnvalue_3 (ScalableFloat|GetValueasInteger _numfruitstogrow_uncommon))
  (bind _numfruitstogrow_rare (Variables|Default|GetNumFruitsToGrow_Rare))
  (bind _returnvalue_4 (ScalableFloat|GetValueasInteger _numfruitstogrow_rare))
  (bind _numfruitstogrow_epic (Variables|Default|GetNumFruitsToGrow_Epic))
  (bind _returnvalue_5 (ScalableFloat|GetValueasInteger _numfruitstogrow_epic))
  (switch Utilities|FlowControl|Switch|SwitchonEFortRarity (Utilities|Enum|BytetoEnumEFortRarity (Math|Conversions|ToByte(Integer) _nurturinglevel))
    (:Common
      (if (>= _returnvalue_2 1)
        (return _returnvalue_2)
        (else
          _)))
    (:Uncommon
      (if (>= _returnvalue_3 1)
        (return _returnvalue_3)
        (else
          _)))
    (:Rare
      (if (>= _returnvalue_4 1)
        (return _returnvalue_4)
        (else
          _)))
    (:Epic
      (if (>= _returnvalue_5 1)
        (return _returnvalue_5)
        (else
          _)))
    (:Legendary
      (if (>= _returnvalue 1)
        (return _returnvalue)
        (else
          _)))
    (:Mythic
      _)
    (:Transcendent
      _)
    (:Unattainable
      _))
  (return (ScalableFloat|GetValueasInteger (Variables|Default|GetNumFruitsToGrow))))

(fn GetScalableFloatTime
  (params (ScalableFloat :type Scalable%20Float%20Structure))
  (if (Variables|Debug|GetbUseShortTimes)
    (return 1.0)
    (else
      (return (ScalableFloat|GetValueatLevel ScalableFloat)))))

(fn GetUTCTimeWithOffsetAdded
  (if (or (>= (Variables|Debug|GetDayOffset) 1) (>= (Variables|Debug|GetMinuteOffset) 1))
    (Variables|SetDebugTimespanOffset_Days (Math|Timespan|MakeTimespan (Variables|Debug|GetDayOffset) 0 (Variables|Debug|GetMinuteOffset)))
    (return (Math.DateTime|AddTimetoDate (CallFunction|GetEventManagerCalendarTime) (Variables|Debug|GetDebugTimespanOffset)))
    (else
      (return (CallFunction|GetEventManagerCalendarTime)))))

(fn GetMarkerDetailsId
  (return "MarkerDetails.RealitySapling"))

(fn GetRarityNameAndColor
  (bind (_owningplayercontroller _location _rotation _nurturinglevel) (Utilities|Struct|BreakRealitySaplingPersistentData (Variables|RealitySapling|GetPersistentSaplingData)))
  (bind _returnvalue (Fortnite|GetRarityData (Utilities|Enum|BytetoEnumEFortRarity (Math|Conversions|ToByte(Integer) _nurturinglevel))))
  (return (Utilities|Struct|BreakFortRarityItemData _returnvalue) (.Color2 (Utilities|Struct|BreakFortRarityItemData _returnvalue))))

(fn GetSaplingAgeVariant
  (bind _saplingbpreppeddata_saplingage_4_5368fd2741dc1813a345149d36ca3e98 (Utilities|Struct|BreakSaplingBPReppedData (Variables|GetSaplingBPReppedData_SaplingAge_4_5368FD2741DC1813A345149D36CA3E98)))
  (return _saplingbpreppeddata_saplingage_4_5368fd2741dc1813a345149d36ca3e98))

(fn InitLocalPersistentSaplingDataFruitStates
  (bind (_persistentsaplingdata_owningplayercontroller _persistentsaplingdata_location _persistentsaplingdata_rotation _persistentsaplingdata_nurturinglevel _persistentsaplingdata_maturitylevel _persistentsaplingdata_plantingtime _persistentsaplingdata_lastweedingtime _persistentsaplingdata_fruitstates) (Utilities|Struct|BreakRealitySaplingPersistentData (Variables|GetPersistentSaplingData_OwningPlayerController)))
  (Variables|Default|SetPersistent%20Sapling%20Data%20Fruit%20States _persistentsaplingdata_fruitstates)
  (while (< (Utilities|Array|Length (Variables|Default|GetPersistent%20Sapling%20Data%20Fruit%20States)) (ScalableFloat|GetValueasInteger (Variables|Default|GetNumFruitsToGrow)))
    (Utilities|Array|Add (Variables|Default|GetPersistent%20Sapling%20Data%20Fruit%20States))))

(fn IsDead
  (bind (_saplingbpreppeddata_saplingage_4_5368fd2741dc1813a345149d36ca3e98 _saplingbpreppeddata_revivecount_5_ec5af65440fb691c7b8724bd7cdfa3e1) (Utilities|Struct|BreakSaplingBPReppedData (Variables|GetSaplingBPReppedData_SaplingAge_4_5368FD2741DC1813A345149D36CA3E98)))
  (return (== (Math|Integer|%25%28Integer%29 _saplingbpreppeddata_revivecount_5_ec5af65440fb691c7b8724bd7cdfa3e1 2) 1)))

(fn IsOkToAutoping
  (bind _owning_player_controller_repped (Variables|Default|GetOwning%20Player%20Controller%20Repped)
    (:then
      (Utilities|IsValid (Actor|GetComponentbyClass _owning_player_controller_repped "/Script/RealitySeedGameplayRuntime.FortControllerComponent_RealitySapling")
        (:Is%20Valid
          (if (.ReturnValue (Class|FortControllerComponentRealitySapling|GetFirstOwnedSaplingData (Actor|GetComponentbyClass _owning_player_controller_repped "/Script/RealitySeedGameplayRuntime.FortControllerComponent_RealitySapling")))
            (return (Math|Boolean|ANDBoolean (not (CallFunction|IsDead)) (not (Variables|Default|GetSaplingTransplanted)) (.bNeedsWeeding (Utilities|Struct|BreakOwnedRealitySaplingData (Class|FortControllerComponentRealitySapling|GetFirstOwnedSaplingData (Actor|GetComponentbyClass _owning_player_controller_repped "/Script/RealitySeedGameplayRuntime.FortControllerComponent_RealitySapling"))))))
            (else
              _)))
        (:Is%20Not%20Valid
          _)))
    (:else
      _))
  (return false))

(fn KillWeeds
  (CallFunction|OnWeedsKilled)
  (for _array_element (Variables|Default|GetServerWeeds)
    (bind _asbga_weed_simple (Utilities|Casting|CastToBGA_Weed_Simple _array_element)
      (:then
        (Class|BGAWeedSimple|ScaleOutAndDie _asbga_weed_simple)
        (Networking|ForceNetUpdate _asbga_weed_simple))
      (:CastFailed
        (Actor|DestroyActor)))))

(fn MarkSapling
  (params (MarkOnClient :type Boolean))
  (bind _owning_player_controller_repped (Variables|Default|GetOwning%20Player%20Controller%20Repped))
  (bind _returnvalue (Class|FortPlayerController|GetMarkerComponent _owning_player_controller_repped))
  (bind _self self)
  (Utilities|IsValid _owning_player_controller_repped
    (:Is%20Valid
      (Utilities|IsValid (MarkerComponent|FindMarkerbyMarkedActor _returnvalue self)
        (:Is%20Not%20Valid
          (if MarkOnClient
            (Fortnite|RegisterSoundIndicatorPlacementContext (Variables|Default|GetOwning%20Player%20Controller%20Repped) _self "0, 0, 0" false)
            (else
              (MarkerComponent|MarkActoronServer _returnvalue _self false false))))))))

(fn NotifyFruitStateChanged
  (params (Fruit :type Tiered%20Reality%20Fruit%20Athena%20Object%20Reference))
  (Networking|SwitchHasAuthority
    (:Authority
      (Utilities|FlowControl|Sequence
        (:then_0
          (if (not (Variables|/RealitySeedGameplay/items/RealitySaplingFruits/Tiered_RealityFruit_Athena.Tiered_RealityFruit_Athena_C|GetIsMythicFruit Fruit))
            (for _array_element (Variables|Default|GetServerSpawnedFruitSlot)
              (bind (_array_element_slot_4_56babd8945fe8b0c64a28cbc0719d0a6 _array_element_fruit_5_dcca3b6b43057342b19a728973abbf7e) (Utilities|Struct|BreakServerSpawnedFruitSlot _array_element))
              (if (== _array_element_fruit_5_dcca3b6b43057342b19a728973abbf7e Fruit)
                (if (not (Variables|Default|GetOverrride%20Sapling%20Persistence))
                  (CallFunction|SetupPersistentSaplingFruitStates))))))
        (:then_1
          (CallFunction|UpdateControllerCompSaplingDataStruct) ; Update struct on controller component for UI purposes
        )
        (:then_2
          (CallFunction|CheckForTooManyFruitsAtRarityLevel false))))))

(fn NotifyFruitsUpgraded
  (CallFunction|CheckForTooManyFruitsAtRarityLevel false)
  (for _array_element (Variables|Default|GetServerSpawnedFruitSlot)
    (bind (_array_element_slot_4_56babd8945fe8b0c64a28cbc0719d0a6 _array_element_fruit_5_dcca3b6b43057342b19a728973abbf7e) (Utilities|Struct|BreakServerSpawnedFruitSlot _array_element))
    (Utilities|IsValid _array_element_fruit_5_dcca3b6b43057342b19a728973abbf7e
      (:Is%20Valid
        (Variables|/RealitySeedGameplay/items/RealitySaplingFruits/Tiered_RealityFruit_Athena.Tiered_RealityFruit_Athena_C|SetUseShortTimes true _array_element_fruit_5_dcca3b6b43057342b19a728973abbf7e)
        (Class|TieredRealityFruitAthena|IncrementGrowthState _array_element_fruit_5_dcca3b6b43057342b19a728973abbf7e))))
  (Networking|SwitchHasAuthority
    (:Authority
      (CallFunction|UpdateControllerCompSaplingDataStruct) ; Update struct on controller component for UI purposes
    )))

(fn NotifyGettingTransplanted
  (Variables|Default|SetDestroyedForTransplant true)
  (Networking|ForceNetUpdate)
  (Utilities|FlowControl|Sequence
    (:then_0
      (for _array_element (Variables|Default|GetServerSpawnedFruitSlot)
        (bind (_array_element_slot_4_56babd8945fe8b0c64a28cbc0719d0a6 _array_element_fruit_5_dcca3b6b43057342b19a728973abbf7e) (Utilities|Struct|BreakServerSpawnedFruitSlot _array_element))
        (Utilities|IsValid _array_element_fruit_5_dcca3b6b43057342b19a728973abbf7e
          (:Is%20Valid
            (Class|TieredRealityFruitAthena|ScaleOutAndDie _array_element_fruit_5_dcca3b6b43057342b19a728973abbf7e)
            (Networking|ForceNetUpdate _array_element_fruit_5_dcca3b6b43057342b19a728973abbf7e)))))
    (:then_1
      (CallFunction|KillWeeds))
    (:then_2
      (for _array_element_2 (Variables|Default|GetServerConsumables)
        (bind _asbuilding_gameplay_actor_consumable (Utilities|Casting|CastToBuildingGameplayActorConsumable _array_element_2)
          (:then
            (bind _ascbga_silky_moose_scale_in (Utilities|Casting|CastToCBGA_SilkyMoose_ScaleIn _asbuilding_gameplay_actor_consumable)
              (:then
                (Class|CBGASilkyMooseScaleIn|ScaleOutAndDie _ascbga_silky_moose_scale_in)
                (Networking|ForceNetUpdate _ascbga_silky_moose_scale_in))
              (:CastFailed
                (Gameplay|SilentDie _asbuilding_gameplay_actor_consumable false))))
          (:CastFailed
            (Actor|DestroyActor)))))
    (:then_3
      (Gameplay|Die))))

(fn NotifySaplingFruitDestroyed
  (params (Destroyed%20Fruit%20Actor :type Actor%20Object%20Reference))
  (CallFunction|OnFruitDestroyed Destroyed%20Fruit%20Actor)
  (Networking|SwitchHasAuthority
    (:Authority
      (CallFunction|UpdateControllerCompSaplingDataStruct) ; Update struct on controller component for UI purposes
    )))

(fn OnConsumableDestroyed
  (params (Consumable :type Actor%20Object%20Reference))
  (if (not (Variables|Default|GetDestroyedForTransplant))
    (Utilities|Array|RemoveItem (Variables|Default|GetServerConsumables) Consumable)))

(fn OnSaplingBabyPlanted
  ;[1 1 1] Stub - Functionality to be overridden in child.
  (Development|PrintString "OnSaplingBabyPlanted")
  (Audio|SpawnSoundatLocation (Variables|Default|GetPlantedSound) (Transformation|GetWorldLocation (Variables|Debug|GetStaticMeshComponent)))
  (if (and (not (CallFunction|IsDead)) (not (Variables|Default|GetSaplingTransplanted)))
    (Default|MarkSapling true)))

(fn OnSaplingBelongstoLocalPlayer)

(fn OnSaplingNewlyTransplanted
  ;[1 1 1] Stub: For functionality to be overrridden in child
  (Development|PrintString "OnSaplingNewlyTransplanted" (Variables|Debug|GetDebug) (Variables|Debug|GetDebug)))

(fn OnTurnedIntoAGhost
  ;[1 1 1] Stub - functionality to be implemented in child
  (Variables|Default|SetIsGhost true))

(fn OnWeedDestroyed
  (params (Weed :type Actor%20Object%20Reference))
  (bind _serverweeds (Variables|Default|GetServerWeeds))
  (bind _returnvalue (Utilities|Array|Length _serverweeds))
  (if (not (Math|Boolean|ORBoolean (Variables|Default|GetDestroyedForTransplant) (Variables|Default|GetDestroyedFromAging) (CallFunction|IsDead)))
    (bind _returnvalue_2 (Utilities|Array|RemoveItem _serverweeds Weed))
    (Variables|Default|SetNumCurrentWeeds _returnvalue)
    (if (and _returnvalue_2 (<= _returnvalue 0))
      (Variables|Default|SetSaplingNurturedCountThisMatch (+ (Variables|Default|GetSaplingNurturedCountThisMatch) 1))
      (Variables|Persistent|SetNeedsWeeding false)
      (CallFunction|StartAddFruitTimerIfNeeded)
      (Networking|ForceNetUpdate)
      (elif _returnvalue_2
        (Utilities|Time|SetTimerbyFunctionName 0 "Sapling Partially Nurtured" (ScalableFloat|GetValueatLevel (Variables|Default|GetDelayForPartialNurtureCelebration)))))))

(fn OnRep_DestroyedForTransplant
  (CallFunction|OnSaplingTransplantingAway)
  (Default|CallSaplingDestroyedforTransplantEvent))

(fn OnRep_HasDied
  (Utilities|FlowControl|Sequence
    (:then_0
      (if (Variables|Default|GetHasDied)
        (Networking|SwitchHasAuthority
          (:Authority
            (CallFunction|KillWeeds)))))
    (:then_1
      (Default|CallSaplingUpdateEvent self))))

(fn OnRep_NeedsWeeding
  (Default|CallSaplingUpdateEvent self))

(fn OnRep_NumCurrentWeeds
  (Default|CallSaplingUpdateEvent self)
  (Networking|SwitchHasAuthority
    (:Authority
      (CallFunction|UpdateControllerCompSaplingDataStruct) ; Update struct on controller component for UI purposes
    )))

(fn OnRep_OverrideIsLocalPlayerSapling
  (if (Variables|Default|GetOverrideIsLocalPlayerSapling)
    (CallFunction|OnSaplingBelongsToLocalPlayer)))

(fn OnRep_OwningPlayerControllerRepped
  (Networking|SwitchHasAuthority
    (:Authority
      (for _array_element (Variables|Default|GetServerWeeds)
        (bind _asbga_weed_simple (Utilities|Casting|CastToBGA_Weed_Simple _array_element))
        (Variables|/RealitySeedGameplay/Environment/Foliage/BGA_Weed_Simple.BGA_Weed_Simple_C|SetOwning%20Player%20Controller%20Repped (Variables|Default|GetOwning%20Player%20Controller%20Repped) _asbga_weed_simple)))
    (:Remote
      (if (!= (Game|GetLocalPlayerControllerID (Variables|Default|GetOwning%20Player%20Controller%20Repped)) -1)
        (CallFunction|OnSaplingBelongsToLocalPlayer)))))

(fn OnRep_OwningPlayerStateRepped)

(fn OnRep_SaplingBPReppedData
  (bind (_saplingbpreppeddata_saplingage_4_5368fd2741dc1813a345149d36ca3e98 _saplingbpreppeddata_revivecount_5_ec5af65440fb691c7b8724bd7cdfa3e1) (Utilities|Struct|BreakSaplingBPReppedData (Variables|GetSaplingBPReppedData_SaplingAge_4_5368FD2741DC1813A345149D36CA3E98)))
  (if (!= (Variables|Default|GetReviveCount_unrepped) _saplingbpreppeddata_revivecount_5_ec5af65440fb691c7b8724bd7cdfa3e1)
    (Variables|Default|SetHasDied false)
    (Variables|Default|SetReviveCount_unrepped _saplingbpreppeddata_revivecount_5_ec5af65440fb691c7b8724bd7cdfa3e1))
  (CallFunction|InitializeSapling)
  (Default|CallSaplingUpdateEvent self))

(fn OnRep_SaplingNurturedCountThisMatch
  (bind (_persistentsaplingdata_owningplayercontroller _persistentsaplingdata_location _persistentsaplingdata_rotation _persistentsaplingdata_nurturinglevel _persistentsaplingdata_maturitylevel _persistentsaplingdata_plantingtime _persistentsaplingdata_lastweedingtime _persistentsaplingdata_fruitstates) (Utilities|Struct|BreakRealitySaplingPersistentData (Variables|GetPersistentSaplingData_OwningPlayerController)))
  (bind _saplingnurturedcountthismatch (Variables|Default|GetSaplingNurturedCountThisMatch))
  (Utilities|FlowControl|Sequence
    (:then_0
      (if (== (Variables|Default|GetPrevNumTimesWeeded) -1)
        ;[1 1 1] First Time
        (Variables|Default|SetPrevNumTimesWeeded _saplingnurturedcountthismatch)
        (Variables|Persistent|SetNumTimesWeeded _saplingnurturedcountthismatch)
        (Networking|SwitchHasAuthority
          (:Remote
            (Default|CallSaplingUpdateEvent self)))
        (elif (> (Variables|Default|GetSaplingNurturedCountThisMatch) (Variables|Default|GetPrevNumTimesWeeded))
          (Variables|Persistent|SetNumTimesWeeded (+ (Variables|Persistent|GetNumTimesWeeded) (- (Variables|Default|GetSaplingNurturedCountThisMatch) (Variables|Default|GetPrevNumTimesWeeded))))
          (Variables|Persistent|SetLastWeededDateTime (CallFunction|GetUTCTimewithOffsetAdded))
          (if (not (Variables|Default|GetOverrride%20Sapling%20Persistence))
            (Variables|SetPersistentSaplingData_OwningPlayerController (Utilities|Struct|MakeRealitySaplingPersistentData _persistentsaplingdata_owningplayercontroller _persistentsaplingdata_location _persistentsaplingdata_rotation (Variables|Persistent|GetNumTimesWeeded) _persistentsaplingdata_maturitylevel _persistentsaplingdata_plantingtime (Variables|Persistent|GetLastWeededDateTime) _persistentsaplingdata_fruitstates)))
          (Variables|Default|SetPrevNumTimesWeeded (Variables|Default|GetSaplingNurturedCountThisMatch))
          (Utilities|Time|SetTimerbyFunctionName 0 "OnSaplingNurtured" (ScalableFloat|GetValueatLevel (Variables|Default|GetDelayForNurtureCelebration)))
          (Networking|SwitchHasAuthority
            (:Remote
              (Default|CallSaplingUpdateEvent self))))))
    (:then_1
      (Networking|SwitchHasAuthority
        (:Authority
          (CallFunction|UpdateControllerCompSaplingDataStruct) ; Update struct on controller component for UI purposes
        )))))

(fn OnRep_ServerSpawnedFruitSlot
  (Default|CallSaplingUpdateEvent self))

(fn SaplingPartiallyNurtured
  (CallFunction|OnSaplingPartiallyNurtured)
  (Networking|ForceNetUpdate))

(fn SetSaplingMeshActive
  (params (InStaticMeshComp :type Static%20Mesh%20Component%20Object%20Reference) (bVisible :type Boolean) (bActiveCollision :type Boolean) (OriginalMaterial :type Material%20Instance%20Object%20Reference) (DeadMaterial :type Material%20Instance%20Object%20Reference))
  (Rendering|SetVisibility InStaticMeshComp bVisible false)
  (if bVisible
    (if (not (CallFunction|IsDead))
      (Rendering|Material|SetMaterial InStaticMeshComp 0 OriginalMaterial)
      (else
        (Rendering|Material|SetMaterial InStaticMeshComp 0 DeadMaterial)))
    (Variables|Default|SetActive%20Static%20Mesh%20Comp InStaticMeshComp))
  (if (and bActiveCollision (not (CallFunction|IsDead)))
    (Collision|SetCollisionEnabled InStaticMeshComp "QueryAndPhysics")
    (Collision|SetCollisionProfileName InStaticMeshComp "FortBuildingMeshPhysics" true :!class /Script/Engine.PrimitiveComponent)
    (else
      (Collision|SetCollisionEnabled InStaticMeshComp "NoCollision"))))

(fn SetNeedsWeeding
  (Variables|Persistent|SetNeedsWeeding true))

(fn SetupPersistentSaplingFruitStates
  ;[1 1 1] Starts Timer for Next Tick
  (Utilities|Time|SetTimerforNextTickbyFunctionName 0 "Setup Persistent Sapling Fruit States Immediate"))

(fn SetupPersistentSaplingFruitStatesImmediate
  (bind _numfruitstogrow (Variables|Default|GetNumFruitsToGrow))
  (bind _returnvalue (ScalableFloat|GetValueasInteger _numfruitstogrow))
  (bind (_persistentsaplingdata_owningplayercontroller _persistentsaplingdata_location _persistentsaplingdata_rotation _persistentsaplingdata_nurturinglevel _persistentsaplingdata_maturitylevel _persistentsaplingdata_plantingtime _persistentsaplingdata_lastweedingtime _persistentsaplingdata_fruitstates) (Utilities|Struct|BreakRealitySaplingPersistentData (Variables|GetPersistentSaplingData_OwningPlayerController)))
  (bind _persistent_sapling_data_fruit_states (Variables|Default|GetPersistent%20Sapling%20Data%20Fruit%20States))
  (bind (_persistentsaplingdata_owningplayercontroller_2 _persistentsaplingdata_location_2 _persistentsaplingdata_rotation_2 _persistentsaplingdata_nurturinglevel_2 _persistentsaplingdata_maturitylevel_2 _persistentsaplingdata_plantingtime_2 _persistentsaplingdata_lastweedingtime_2) (Utilities|Struct|BreakRealitySaplingPersistentData (Variables|GetPersistentSaplingData_OwningPlayerController)))
  (bind _persistent_sapling_data_fruit_states_2 (Variables|Default|GetPersistent%20Sapling%20Data%20Fruit%20States))
  (Utilities|Array|Clear (Variables|Default|GetPersistent%20Sapling%20Data%20Fruit%20States))
  (if (Variables|Default|GetOverrride%20Sapling%20Persistence)
    (if (>= _returnvalue 1)
      (Utilities|Array|Add _persistent_sapling_data_fruit_states_2 (Variables|Persistent|GetFruit1State))
      (if (>= _returnvalue 2)
        (Utilities|Array|Add _persistent_sapling_data_fruit_states_2 (Variables|Persistent|GetFruit2State))
        (if (>= _returnvalue 3)
          (Utilities|Array|Add _persistent_sapling_data_fruit_states_2 (Variables|Persistent|GetFruit3State)))))
    (else
      (for _array_element (Variables|Default|GetServerSpawnedFruitSlot)
        (bind (_array_element_slot_4_56babd8945fe8b0c64a28cbc0719d0a6 _array_element_fruit_5_dcca3b6b43057342b19a728973abbf7e) (Utilities|Struct|BreakServerSpawnedFruitSlot _array_element))
        (Utilities|IsValid _array_element_fruit_5_dcca3b6b43057342b19a728973abbf7e
          (:Is%20Valid
            (Utilities|Array|Add _persistent_sapling_data_fruit_states (Math|Conversions|ToInteger(Byte) (Variables|/RealitySeedGameplay/items/RealitySaplingFruits/Tiered_RealityFruit_Athena.Tiered_RealityFruit_Athena_C|GetVariantOrRarity _array_element_fruit_5_dcca3b6b43057342b19a728973abbf7e))))
          (:Is%20Not%20Valid
            (Utilities|Array|Add _persistent_sapling_data_fruit_states))))
      (Variables|SetPersistentSaplingData_OwningPlayerController (Utilities|Struct|MakeRealitySaplingPersistentData _persistentsaplingdata_owningplayercontroller_2 _persistentsaplingdata_location_2 _persistentsaplingdata_rotation_2 _persistentsaplingdata_nurturinglevel_2 _persistentsaplingdata_maturitylevel_2 _persistentsaplingdata_plantingtime_2 _persistentsaplingdata_lastweedingtime_2 (Variables|Default|GetPersistent%20Sapling%20Data%20Fruit%20States)))
      (CallFunction|UpdateControllerCompSaplingDataStruct) ; Update struct on controller component for UI purposes
      (if (Variables|Debug|GetDebug)
        (Development|PrintString (Utilities|String|Append "Fruit States: Length = " (Utilities|String|ToString(Integer) (Utilities|Array|Length _persistentsaplingdata_fruitstates)) "   [" (Utilities|String|ToString(Integer) (Utilities|Array|Get(aref) _persistentsaplingdata_fruitstates)) "][" (Utilities|String|ToString(Integer) (Utilities|Array|Get(aref) _persistentsaplingdata_fruitstates 1)) "][" (Utilities|String|ToString(Integer) (Utilities|Array|Get(aref) _persistentsaplingdata_fruitstates 2)) "]") (Variables|Debug|GetDebug) (Variables|Debug|GetDebug) "(R=1.000000,G=1.000000,B=1.000000,A=1.000000)" 1e+02)))))

(fn SpawnFruit
  (params (ArrayIndex :type Integer) (bIsMythic :type Boolean))
  (bind (_outtransform_location _outtransform_rotation _outtransform_scale) (Math|Transform|BreakTransform (CallFunction|GetFruitSpawnTransform ArrayIndex)))
  (bind (_outtransform_location_2 _outtransform_rotation_2 _outtransform_scale_2) (Math|Transform|BreakTransform (CallFunction|GetFruitSpawnTransform ArrayIndex)))
  (Development|PrintString (Utilities|String|Append "Spawning Fruit in Slot:" (Utilities|String|ToString(Integer) ArrayIndex) "   State: " (Utilities|String|EnumtoString (Variables|Default|GetInitialSaplingFruitVariant))) (Variables|Debug|GetDebug) (Variables|Debug|GetDebug) "(R=0.000000,G=0.660000,B=1.000000,A=1.000000)" 2e+02)
  (if (Math|Boolean|ORBoolean bIsMythic (Utilities|Enum|Equal(Enum) (Variables|Default|GetInitialSaplingFruitVariant) "NewEnumerator7") (Utilities|Enum|Equal(Enum) (Variables|Default|GetInitialSaplingFruitVariant) "NewEnumerator8"))
    (bind _returnvalue (Game|SpawnActorfromClass "/RealitySeedGameplay/items/RealitySaplingFruits/Tiered_RealityFruit_Athena_Mythic.Tiered_RealityFruit_Athena_Mythic_C" (Math|Transform|MakeTransform _outtransform_location_2 _outtransform_rotation_2 _outtransform_scale_2) "AlwaysSpawn" "MultiplyWithRoot" 0 (Variables|Default|GetInitialSaplingFruitVariant) true ArrayIndex (Utilities|Struct|MakeFruitReppedData (Variables|Default|GetOwning%20Player%20Controller%20Repped) self -1 (Variables|Default|GetOverrideIsLocalPlayerSapling)) (Variables|Gameplay|GetDamagerOwner) 0 (Variables|Actor|GetInstigator)))
    (Utilities|Array|SetArrayElem (Variables|Default|GetServerSpawnedFruitSlot) ArrayIndex (Utilities|Struct|MakeServerSpawnedFruitSlot ArrayIndex _returnvalue) false)
    (return _returnvalue)
    (else
      (bind _returnvalue_2 (Game|SpawnActorfromClass "/RealitySeedGameplay/items/RealitySaplingFruits/Tiered_RealityFruit_Athena.Tiered_RealityFruit_Athena_C" (Math|Transform|MakeTransform _outtransform_location _outtransform_rotation _outtransform_scale) "AlwaysSpawn" "MultiplyWithRoot" 0 (Variables|Default|GetInitialSaplingFruitVariant) true ArrayIndex (Utilities|Struct|MakeFruitReppedData (Variables|Default|GetOwning%20Player%20Controller%20Repped) self -1 (Variables|Default|GetOverrideIsLocalPlayerSapling)) (Variables|Gameplay|GetDamagerOwner) 0 (Variables|Actor|GetInstigator)))
      (Utilities|Array|SetArrayElem (Variables|Default|GetServerSpawnedFruitSlot) ArrayIndex (Utilities|Struct|MakeServerSpawnedFruitSlot ArrayIndex _returnvalue_2) false)
      (return _returnvalue_2))))

(fn StartAddFruitTimerifNeeded
  (if (Variables|Persistent|GetNeedsWeeding)
    _
    (elif (Utilities|Time|IsValidTimerHandle (Variables|Default|GetCreateFruitTimerHandle))
      _
      (else
        (bind _returnvalue (Utilities|Time|SetTimerforNextTickbyFunctionName 0 "CreateFruits"))
        (Variables|Default|SetCreateFruitTimerHandle _returnvalue)))))

(fn UpdateControllerCompSaplingDataStruct
  (bind (_persistentsaplingdata_owningplayercontroller _persistentsaplingdata_location _persistentsaplingdata_rotation _persistentsaplingdata_nurturinglevel) (Utilities|Struct|BreakRealitySaplingPersistentData (Variables|GetPersistentSaplingData_OwningPlayerController)))
  (bind (_lastweededdatetime_year _lastweededdatetime_month _lastweededdatetime_day) (Math|DateTime|BreakDateTime (Variables|GetLastWeededDateTime_Year)))
  (bind (_persistentsaplingdata_owningplayercontroller_2 _persistentsaplingdata_location_2 _persistentsaplingdata_rotation_2 _persistentsaplingdata_nurturinglevel_2 _persistentsaplingdata_maturitylevel _persistentsaplingdata_plantingtime _persistentsaplingdata_lastweedingtime _persistentsaplingdata_fruitstates) (Utilities|Struct|BreakRealitySaplingPersistentData (Variables|GetPersistentSaplingData_OwningPlayerController)))
  ;[0.4518 1 0.8693] Update the FOwnedSaplingData struct on the controller component. This struct will replicate for UI purposes where the sapling cannot be used directly due to relevancy.
  (Networking|SwitchHasAuthority
    (:Authority
      (bind _owning_player_controller_repped (Variables|Default|GetOwning%20Player%20Controller%20Repped)
        (:then
          (Utilities|IsValid (Actor|GetComponentbyClass _owning_player_controller_repped "/Script/RealitySeedGameplayRuntime.FortControllerComponent_RealitySapling")
            (:Is%20Valid
              (if (.ReturnValue (Class|FortControllerComponentRealitySapling|GetFirstOwnedSaplingData (Actor|GetComponentbyClass _owning_player_controller_repped "/Script/RealitySeedGameplayRuntime.FortControllerComponent_RealitySapling")))
                (for (_array_element _array_index) _persistentsaplingdata_fruitstates
                  (if (< _array_index (ScalableFloat|GetValueasInteger (Variables|Default|GetNumFruitsToGrow)))
                    (Utilities|Array|Add (LocalVariables|GetLocalFruitStates) (Math|Conversions|ToByte(Integer) _array_element))))
                (Class|FortControllerComponentRealitySapling|SetFirstOwnedSaplingData (Actor|GetComponentbyClass _owning_player_controller_repped "/Script/RealitySeedGameplayRuntime.FortControllerComponent_RealitySapling") (Utilities|Struct|MakeOwnedRealitySaplingData self (.Location (Utilities|Struct|BreakOwnedRealitySaplingData (Class|FortControllerComponentRealitySapling|GetFirstOwnedSaplingData (Actor|GetComponentbyClass _owning_player_controller_repped "/Script/RealitySeedGameplayRuntime.FortControllerComponent_RealitySapling")))) (.bHasSaplingOnAuthority (Utilities|Struct|BreakOwnedRealitySaplingData (Class|FortControllerComponentRealitySapling|GetFirstOwnedSaplingData (Actor|GetComponentbyClass _owning_player_controller_repped "/Script/RealitySeedGameplayRuntime.FortControllerComponent_RealitySapling")))) (LocalVariables|GetLocalFruitStates) (Math|Conversions|ToByte(Integer) _persistentsaplingdata_nurturinglevel) (> (Utilities|Array|Length (Variables|Default|GetServerWeeds)) 0) (CallFunction|IsDead) (Variables|Default|GetEnemyAttacking) (Variables|Default|GetEnemyLooting) (Math|DateTime|MakeDateTime _lastweededdatetime_year _lastweededdatetime_month (+ _lastweededdatetime_day 1))))
                (Development|PrintString "Updated controller component struct" (Variables|Debug|GetDebug) (Variables|Debug|GetDebug))))))
        (:else)))))

(fn UpdateSafeZoneStatus
  (params (bInSafeZone :type Boolean))
  (if (!= bInSafeZone (Variables|Default|GetIsInSafeZone))
    (Variables|Default|SetIsInSafeZone bInSafeZone)
    (Default|CallSaplingUpdateEvent self)))

(fn ConstructionScript)
