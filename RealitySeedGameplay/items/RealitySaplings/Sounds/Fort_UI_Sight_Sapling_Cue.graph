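# Sound cue for a Reality Sapling sight UI ping, played as a non-critical in-game UI sound with ping attenuation.
# A mixer layers a shared BigPositive sight sound with a slightly quieter Sapling-specific ping that starts after a short fixed delay.
# Modulators hold both layers at fixed pitch.
# A concurrency override allows one instance at a time, stopping the oldest when a new one plays.
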SoundCue /RealitySeedGameplay/items/RealitySaplings/Sounds/Fort_UI_Sight_Sapling_Cue
root=Mixer_0
nodes [
 {id=Delay_0, children [Modulator_1], properties {delayMax=0.15, delayMin=0.15}}
 {id=Mixer_0, children [Modulator_0,Delay_0], properties {inputVolume [1,0.8]}}
 {id=Modulator_0, children [WavePlayer_0], properties {pitchMax=1, pitchMin=1, volumeMax=1, volumeMin=1}}
 {id=Modulator_1, children [WavePlayer_1], properties {pitchMax=1, pitchMin=1}}
 {id=WavePlayer_0, properties {soundWaveAssetPtr=/Game/Sounds/Fort_UI_Sounds/Gameplay/Ping/Fort_UI_Sight_BigPositive_03}}
 {id=WavePlayer_1, properties {soundWaveAssetPtr=&/Sapling_UI_Ping}}
]
properties {
 attenuationSettings=/Game/Sounds/Attenuation/Fort_UI_Ping_Ingame_Att
 bOverrideConcurrency=true
 maxDistance=2097152
 soundClassObject=&Fort_UI_InGame_NonCritical
 volumeMultiplier=1
 concurrencyOverrides {maxCount=1, resolutionRule=StopOldest}
}
