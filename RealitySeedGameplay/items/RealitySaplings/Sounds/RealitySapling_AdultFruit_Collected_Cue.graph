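# Sound cue for collecting fruit from an adult Reality Sapling item.
# A random node picks evenly among five fruit-collect wave variations and feeds a modulator at the root.
# It plays at reduced volume in the gadgets/items sound class, using the Reality Pods prop attenuation asset for spatial falloff.
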SoundCue /RealitySeedGameplay/items/RealitySaplings/Sounds/RealitySapling_AdultFruit_Collected_Cue
root=Modulator_0
nodes [
 {id=Modulator_0, children [Random_0]}
 {id=Random_0, children [WavePlayer_4,WavePlayer_5,WavePlayer_6,WavePlayer_7,WavePlayer_8], properties {weights [1,1,1,1,1]}}
 {id=WavePlayer_4, properties {soundWaveAssetPtr=&/Sapling_FruitCollect_01}}
 {id=WavePlayer_5, properties {soundWaveAssetPtr=&/Sapling_FruitCollect_02}}
 {id=WavePlayer_6, properties {soundWaveAssetPtr=&/Sapling_FruitCollect_03}}
 {id=WavePlayer_7, properties {soundWaveAssetPtr=&/Sapling_FruitCollect_04}}
 {id=WavePlayer_8, properties {soundWaveAssetPtr=&/Sapling_FruitCollect_05}}
]
properties {
 attenuationSettings=/Game/Sounds/Attenuation/Fort_Prop_RealityPods_Att
 bOverrideAttenuation=true
 maxDistance=4000
 soundClassObject=&Fort_Gadgets_Items
 volumeMultiplier=0.6
}
