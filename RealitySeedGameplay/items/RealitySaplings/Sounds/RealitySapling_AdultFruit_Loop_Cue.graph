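# Looping sound cue for the adult fruit state of Reality Saplings in the RealitySeedGameplay items.
# A looping wave player plays the Sapling_AdultFruit_Loop wave through a modulator that applies a fixed pitch raise and volume boost, with no random variation.
# The cue uses the Fort_Gadgets_Items sound class and overrides attenuation with the shared Reality Pods prop attenuation asset.
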SoundCue /RealitySeedGameplay/items/RealitySaplings/Sounds/RealitySapling_AdultFruit_Loop_Cue
root=Modulator_0
nodes [
 {id=Modulator_0, children [WavePlayer_3], properties {pitchMax=1.5, pitchMin=1.5, volumeMax=1.2, volumeMin=1.2}}
 {id=WavePlayer_3, properties {bLooping=true, soundWaveAssetPtr=&/Sapling_AdultFruit_Loop}}
]
properties {
 attenuationSettings=/Game/Sounds/Attenuation/Fort_Prop_RealityPods_Att
 bOverrideAttenuation=true
 maxDistance=4000
 soundClassObject=&Fort_Gadgets_Items
 volumeMultiplier=1
}
