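# Looping ambient sound cue for Reality Saplings, a Reality Seed gameplay item in the gadgets and items sound class.
# A looping wave player for the non-player sapling 3D loop feeds a modulator that pins pitch at double rather than randomizing it.
# Inline attenuation overrides the assigned RealityPods attenuation asset for a short-range falloff, and inline concurrency caps simultaneous instances at a few.
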SoundCue /RealitySeedGameplay/items/RealitySaplings/Sounds/RealitySapling_Amb_Loop_Cue
root=Modulator_0
nodes [
 {id=Modulator_0, children [WavePlayer_0], properties {pitchMax=2, pitchMin=2}}
 {
  id=WavePlayer_0
  properties {bLooping=true, soundWaveAssetPtr=&/Sapling_NonPlayerSapling_3D_Loop}
 }
]
properties {
 attenuationSettings=/Game/Sounds/Attenuation/Fort_Prop_RealityPods_Att
 bOverrideAttenuation=true
 bOverrideConcurrency=true
 maxDistance=500
 soundClassObject=&Fort_Gadgets_Items
 volumeMultiplier=1.5
 attenuationOverrides {falloffDistance=400, attenuationShapeExtents {x=100}}
 concurrencyOverrides {maxCount=3}
 soundSubmixSends [{minSendLevel=0, sendLevel=0, soundSubmix=null}]
}
