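# Sound cue that plays an impact sound when a Reality Sapling item is damaged.
# A Random node picks one of four equally weighted sapling impact waves and feeds a Modulator that varies pitch on each play.
# The cue slightly lowers overall pitch and volume and routes to the gadgets and items sound class.
# Inline attenuation overrides limit its audible range, although a Reality Pods attenuation asset is also assigned.
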SoundCue /RealitySeedGameplay/items/RealitySaplings/Sounds/RealitySapling_Damaged_Cue
root=Modulator_0
nodes [
 {id=Modulator_0, children [Random_0], properties {pitchMax=1.1}}
 {id=Random_0, children [WavePlayer_1,WavePlayer_2,WavePlayer_4,WavePlayer_5], properties {weights [1,1,1,1]}}
 {id=WavePlayer_1, properties {soundWaveAssetPtr=&/Sapling_Impact_01}}
 {id=WavePlayer_2, properties {soundWaveAssetPtr=&/Sapling_Impact_02}}
 {id=WavePlayer_4, properties {soundWaveAssetPtr=&/Sapling_Impact_03}}
 {id=WavePlayer_5, properties {soundWaveAssetPtr=&/Sapling_Impact_04}}
]
properties {
 attenuationSettings=/Game/Sounds/Attenuation/Fort_Prop_RealityPods_Att
 bOverrideAttenuation=true
 maxDistance=2100
 pitchMultiplier=0.9
 soundClassObject=&Fort_Gadgets_Items
 volumeMultiplier=0.8
 attenuationOverrides {falloffDistance=2000, attenuationShapeExtents {x=100}}
}
