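# Sound cue for the Reality Sapling fruit being destroyed.
# A mixer layers a borrowed Stink Flopper fish splat with one of five equally weighted fruit-destroyed variants chosen at random, the latter mixed louder.
# Each layer passes through its own modulator.
# The cue plays quietly in the gadget items sound class, using Reality Pods prop attenuation with overrides enabled.
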SoundCue /RealitySeedGameplay/items/RealitySaplings/Sounds/RealitySapling_Fruit_Destroyed_Cue
root=Mixer_0
nodes [
 {id=Mixer_0, children [Modulator_0,Modulator_1], properties {inputVolume [1,2]}}
 {id=Modulator_0, children [WavePlayer_3]}
 {id=Modulator_1, children [Random_0]}
 {id=Random_0, children [WavePlayer_2,WavePlayer_4,WavePlayer_5,WavePlayer_6,WavePlayer_7], properties {weights [1,1,1,1,1]}}
 {id=WavePlayer_2, properties {soundWaveAssetPtr=&/Sapling_FruitDestroyed_01}}
 {id=WavePlayer_3, properties {soundWaveAssetPtr=/Game/Athena/Sounds/Items/Flopper/StinkFlopper/Stink_Fish_Splats_01}}
 {id=WavePlayer_4, properties {soundWaveAssetPtr=&/Sapling_FruitDestroyed_02}}
 {id=WavePlayer_5, properties {soundWaveAssetPtr=&/Sapling_FruitDestroyed_03}}
 {id=WavePlayer_6, properties {soundWaveAssetPtr=&/Sapling_FruitDestroyed_04}}
 {id=WavePlayer_7, properties {soundWaveAssetPtr=&/Sapling_FruitDestroyed_05}}
]
properties {
 attenuationSettings=/Game/Sounds/Attenuation/Fort_Prop_RealityPods_Att
 bOverrideAttenuation=true
 maxDistance=4000
 soundClassObject=&Fort_Gadgets_Items
 volumeMultiplier=0.3
 attenuationOverrides {stereoSpread=100}
}
