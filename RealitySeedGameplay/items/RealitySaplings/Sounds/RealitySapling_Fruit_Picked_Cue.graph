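# Sound cue played when a Reality Sapling fruit is picked, layering pick, splat and crunch sounds.
# An immediate crunch layer with slight random pitch variation mixes with a briefly delayed layer.
# The delayed layer blends one of four equally weighted random pick variations with a stink-fish splat borrowed from the Flopper item.
# It is classed as a gadget item sound and uses Reality Pods prop attenuation.
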SoundCue /RealitySeedGameplay/items/RealitySaplings/Sounds/RealitySapling_Fruit_Picked_Cue
root=Mixer_0
nodes [
 {id=Delay_0, children [Mixer_1], properties {delayMax=0.08, delayMin=0.08}}
 {id=Mixer_0, children [Delay_0,Modulator_2], properties {inputVolume [1,0.65]}}
 {id=Mixer_1, children [Modulator_0,Modulator_1], properties {inputVolume [1,2]}}
 {id=Modulator_0, children [Random_0], properties {volumeMax=1.5, volumeMin=1.5}}
 {id=Modulator_1, children [WavePlayer_4], properties {volumeMax=0.5, volumeMin=0.5}}
 {id=Modulator_2, children [WavePlayer_6], properties {pitchMax=1.1, pitchMin=0.9, volumeMax=2, volumeMin=2}}
 {id=Random_0, children [WavePlayer_0,WavePlayer_1,WavePlayer_2,WavePlayer_3], properties {weights [1,1,1,1]}}
 {id=WavePlayer_0, properties {soundWaveAssetPtr=&/Sapling_FruitPicked_01}}
 {id=WavePlayer_1, properties {soundWaveAssetPtr=&/Sapling_FruitPicked_02}}
 {id=WavePlayer_2, properties {soundWaveAssetPtr=&/Sapling_FruitPicked_03}}
 {id=WavePlayer_3, properties {soundWaveAssetPtr=&/Sapling_FruitPicked_04}}
 {id=WavePlayer_4, properties {soundWaveAssetPtr=/Game/Athena/Sounds/Items/Flopper/StinkFlopper/Stink_Fish_Splats_01}}
 {id=WavePlayer_6, properties {soundWaveAssetPtr=&/Sapling_FruitPickedCrunch}}
]
properties {
 attenuationSettings=/Game/Sounds/Attenuation/Fort_Prop_RealityPods_Att
 bOverrideAttenuation=true
 maxDistance=4000
 soundClassObject=&Fort_Gadgets_Items
 volumeMultiplier=0.35
}
