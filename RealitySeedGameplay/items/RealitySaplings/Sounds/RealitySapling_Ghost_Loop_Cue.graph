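# Sound cue for the looping ghost sound of the Reality Sapling item, classed under gadget and item sounds.
# A looping Sapling_Ghost wave player feeds a Modulator that pins pitch at a fixed doubled value rather than a random range, and the cue halves the volume.
# Attenuation override is enabled, so the cue's own short-range falloff applies in place of the referenced Reality Pods attenuation asset.
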SoundCue /RealitySeedGameplay/items/RealitySaplings/Sounds/RealitySapling_Ghost_Loop_Cue
root=Modulator_0
nodes [
 {id=Modulator_0, children [WavePlayer_1], properties {pitchMax=2, pitchMin=2}}
 {id=WavePlayer_1, properties {bLooping=true, soundWaveAssetPtr=&/Sapling_Ghost}}
]
properties {
 attenuationSettings=/Game/Sounds/Attenuation/Fort_Prop_RealityPods_Att
 bOverrideAttenuation=true
 maxDistance=900
 soundClassObject=&Fort_Gadgets_Items
 volumeMultiplier=0.5
 attenuationOverrides {falloffDistance=800, attenuationShapeExtents {x=100}}
}
