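# Sound cue that plays the nourished sound for a Reality Sapling item.
# A random node picks among three equally weighted wave variants and feeds a modulator at the root, left at its default settings.
# The cue sets attenuation and concurrency overrides, limiting playback to one instance per owner and blocking new plays while one is active.
# It routes to the gadgets and items sound class at reduced volume.
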SoundCue /RealitySeedGameplay/items/RealitySaplings/Sounds/RealitySapling_Nourished_Cue
root=Modulator_0
nodes [
 {id=Modulator_0, children [Random_0]}
 {id=Random_0, children [WavePlayer_0,WavePlayer_1,WavePlayer_2], properties {weights [1,1,1]}}
 {id=WavePlayer_0, properties {soundWaveAssetPtr=&/Sapling_Nourished_01}}
 {id=WavePlayer_1, properties {soundWaveAssetPtr=&/Sapling_Nourished_02}}
 {id=WavePlayer_2, properties {soundWaveAssetPtr=&/Sapling_Nourished_03}}
]
properties {
 attenuationSettings=/Game/Sounds/Attenuation/Fort_Prop_RealityPods_Att
 bOverrideAttenuation=true
 bOverrideConcurrency=true
 maxDistance=2100
 soundClassObject=&Fort_Gadgets_Items
 volumeMultiplier=0.6
 attenuationOverrides {falloffDistance=2000, attenuationShapeExtents {x=100}}
 concurrencyOverrides {bLimitToOwner=true, maxCount=1, resolutionRule=PreventNew}
}
