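# Sound cue for the nurtured sound of a Reality Sapling item, routed to the gadget items sound class.
# A single wave player feeds a root Modulator that pins pitch at normal, so the sound has no pitch variation.
# The cue lowers overall volume and enables its own attenuation override for falloff, which takes precedence over the referenced Reality Pods attenuation asset.
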SoundCue /RealitySeedGameplay/items/RealitySaplings/Sounds/RealitySapling_Nurtured_Cue
root=Modulator_0
nodes [
 {id=Modulator_0, children [WavePlayer_1], properties {pitchMax=1, pitchMin=1}}
 {id=WavePlayer_1, properties {soundWaveAssetPtr=&/Sapling_Nurtured_01}}
]
properties {
 attenuationSettings=/Game/Sounds/Attenuation/Fort_Prop_RealityPods_Att
 bOverrideAttenuation=true
 maxDistance=2100
 soundClassObject=&Fort_Gadgets_Items
 volumeMultiplier=0.6
 attenuationOverrides {falloffDistance=2000, attenuationShapeExtents {x=100}}
}
