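# Looping Sound Cue for an owned, adult Reality Sapling, providing its positional ambient audio.
# A Mixer layers two looping waves: the mature sapling loop and a mature bugs loop.
# Modulators set fixed volumes without randomization, boosting the sapling layer and lowering the bugs.
# The cue uses the gadget items sound class and overrides its referenced Reality Pods attenuation with its own falloff settings.
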SoundCue /RealitySeedGameplay/items/RealitySaplings/Sounds/RealitySapling_Owned_Adult_Loop_Cue
root=Mixer_1
nodes [
 {id=Mixer_1, children [Modulator_0,Modulator_1], properties {inputVolume [1,0.9]}}
 {id=Modulator_0, children [WavePlayer_0], properties {pitchMax=1, pitchMin=1, volumeMax=1.5, volumeMin=1.5}}
 {id=Modulator_1, children [WavePlayer_1], properties {pitchMax=1, pitchMin=1, volumeMax=0.7, volumeMin=0.7}}
 {
  id=WavePlayer_0
  properties {bLooping=true, soundWaveAssetPtr=&/Sapling_YourSapling_Mature_3D_Loop}
 }
 {
  id=WavePlayer_1
  properties {bLooping=true, soundWaveAssetPtr=&/Sapling_YourSapling_MatureBugs_3D_Loop}
 }
]
properties {
 attenuationSettings=/Game/Sounds/Attenuation/Fort_Prop_RealityPods_Att
 bOverrideAttenuation=true
 maxDistance=1100
 soundClassObject=&Fort_Gadgets_Items
 volumeMultiplier=1.2
 attenuationOverrides {falloffDistance=1000, attenuationShapeExtents {x=100}}
}
