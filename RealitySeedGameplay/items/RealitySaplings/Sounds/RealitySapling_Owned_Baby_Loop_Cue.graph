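# Looping sound cue for an owned baby Reality Sapling, giving the player's own young sapling a continuous 3D sound.
# One looping wave player feeds a modulator whose pitch range is pinned, so pitch does not vary.
# The cue plays at reduced volume under the gadgets and items sound class.
# It references the shared Reality Pods attenuation asset but overrides attenuation with its own shape and falloff.
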SoundCue /RealitySeedGameplay/items/RealitySaplings/Sounds/RealitySapling_Owned_Baby_Loop_Cue
root=Modulator_0
nodes [
 {id=Modulator_0, children [WavePlayer_0], properties {pitchMax=1, pitchMin=1}}
 {
  id=WavePlayer_0
  properties {bLooping=true, soundWaveAssetPtr=&/Sapling_YourSapling_Youth_3D_Loop}
 }
]
properties {
 attenuationSettings=/Game/Sounds/Attenuation/Fort_Prop_RealityPods_Att
 bOverrideAttenuation=true
 maxDistance=1300
 soundClassObject=&Fort_Gadgets_Items
 volumeMultiplier=0.5
 attenuationOverrides {falloffDistance=1200, attenuationShapeExtents {x=100}}
}
