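# Sound cue for a Reality Sapling being removed, playing the Sapling_Wither wave.
# A single wave player feeds a Modulator left at default settings, adding slight random pitch and volume variation on each play.
# The cue runs at reduced volume in the Fort_Gadgets_Items sound class.
# Attenuation override is enabled, so its own falloff and shape values apply instead of the referenced Reality Pods attenuation asset.
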SoundCue /RealitySeedGameplay/items/RealitySaplings/Sounds/RealitySapling_Removed_Cue
root=Modulator_0
nodes [
 {id=Modulator_0, children [WavePlayer_0]}
 {id=WavePlayer_0, properties {soundWaveAssetPtr=&/Sapling_Wither}}
]
properties {
 attenuationSettings=/Game/Sounds/Attenuation/Fort_Prop_RealityPods_Att
 bOverrideAttenuation=true
 maxDistance=1300
 soundClassObject=&Fort_Gadgets_Items
 volumeMultiplier=0.4
 attenuationOverrides {falloffDistance=1200, attenuationShapeExtents {x=100}}
}
