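# Sound cue for transplanting a Reality Sapling, routed to the gadgets and items sound class at reduced volume.
# A mixer layers two modulated branches: a random pick of two plant-seed waves, and a single transplant wave with slight volume variation.
# Attenuation and concurrency are overridden on the cue, limiting it to one instance per owner and preventing new plays while one is active.
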SoundCue /RealitySeedGameplay/items/RealitySaplings/Sounds/RealitySapling_Transplanted_Cue
root=Mixer_0
nodes [
 {id=Mixer_0, children [Modulator_0,Modulator_1], properties {inputVolume [1,1]}}
 {id=Modulator_0, children [Random_0]}
 {id=Modulator_1, children [WavePlayer_2], properties {volumeMax=1.02, volumeMin=0.98}}
 {id=Random_0, children [WavePlayer_0,WavePlayer_1], properties {weights [1,1]}}
 {id=WavePlayer_0, properties {soundWaveAssetPtr=&/Sapling_PlantSeed__01}}
 {id=WavePlayer_1, properties {soundWaveAssetPtr=&/Sapling_PlantSeed__02}}
 {id=WavePlayer_2, properties {soundWaveAssetPtr=&/Sapling_Transplant_01}}
]
properties {
 attenuationSettings=/Game/Sounds/Attenuation/Fort_Prop_RealityPods_Att
 bOverrideAttenuation=true
 bOverrideConcurrency=true
 maxDistance=2600
 soundClassObject=&Fort_Gadgets_Items
 volumeMultiplier=0.5
 attenuationOverrides {falloffDistance=2500, attenuationShapeExtents {x=100}}
 concurrencyOverrides {bLimitToOwner=true, maxCount=1, resolutionRule=PreventNew, retriggerTime=1}
 soundSubmixSends [{minSendLevel=0, sendLevel=0, soundSubmix=null}]
}
