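# Sound cue for the weed-pull sound of a Reality Sapling item, classed under gadget and item sounds.
# A random node picks among four equally weighted weed-pull waves, then feeds a modulator with pitch held fixed.
# Inline attenuation overrides the assigned Reality Pods prop attenuation asset with logarithmic falloff, distance low-pass filtering, listener focus and a distant-reverb submix send.
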SoundCue /RealitySeedGameplay/items/RealitySaplings/Sounds/RealitySapling_WeedPulled_Cue
root=Modulator_0
nodes [
 {id=Modulator_0, children [Random_0], properties {pitchMax=1, pitchMin=1}}
 {id=Random_0, children [WavePlayer_0,WavePlayer_1,WavePlayer_2,WavePlayer_4], properties {weights [1,1,1,1]}}
 {id=WavePlayer_0, properties {soundWaveAssetPtr=&/Sapling_WeedPull_01}}
 {id=WavePlayer_1, properties {soundWaveAssetPtr=&/Sapling_WeedPull_02}}
 {id=WavePlayer_2, properties {soundWaveAssetPtr=&/Sapling_WeedPull_03}}
 {id=WavePlayer_4, properties {soundWaveAssetPtr=&/Sapling_WeedPull_04}}
]
properties {
 attenuationSettings=/Game/Sounds/Attenuation/Fort_Prop_RealityPods_Att
 bOverrideAttenuation=true
 maxDistance=2200
 soundClassObject=&Fort_Gadgets_Items
 volumeMultiplier=0.5
 attenuationOverrides {
  bAttenuateWithLPF=true
  bEnableListenerFocus=true
  bEnableSubmixSends=true
  distanceAlgorithm=Logarithmic
  falloffDistance=2000
  lPFFrequencyAtMax=1000
  lPFFrequencyAtMin=5000
  lPFRadiusMax=400
  lPFRadiusMin=50
  nonFocusAzimuth=45
  nonFocusVolumeAttenuation=0.3
  reverbDistanceMax=1024
  reverbDistanceMin=0
  reverbWetLevelMax=0
  reverbWetLevelMin=1
  stereoSpread=100
  attenuationShapeExtents {x=200}
  submixSendSettings [{
    bEnableHPFCutoff=false
    bEnableLPFCutoff=false
    disableManualSendClamp=false
    hPFCutoff=20
    lPFCutoff=20000
    maxSendDistance=7500
    maxSendLevel=0.25
    minSendDistance=0
    minSendLevel=0
    sendLevel=0.2
    sendLevelControlMethod=Linear
    soundSubmix=&Fortnite_Reverb_Distant
   }]
 }
}
