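# Sound cue for the withering of a Reality Sapling item, playing the Sapling_Wither wave.
# A Modulator at the root feeds from a single wave player, with the modulator left at its default variation settings.
# It uses the Reality Pods prop attenuation, the gadget items sound class and a reduced volume multiplier.
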SoundCue /RealitySeedGameplay/items/RealitySaplings/Sounds/RealitySapling_Withering_Cue
root=Modulator_0
nodes [
 {id=Modulator_0, children [WavePlayer_1]}
 {id=WavePlayer_1, properties {soundWaveAssetPtr=&/Sapling_Wither}}
]
properties {
 attenuationSettings=/Game/Sounds/Attenuation/Fort_Prop_RealityPods_Att
 maxDistance=1600
 soundClassObject=&Fort_Gadgets_Items
 volumeMultiplier=0.6
}
